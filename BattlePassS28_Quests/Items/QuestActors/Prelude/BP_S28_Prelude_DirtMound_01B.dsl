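; Season 28 Prelude dirt mound quest prop that changes visuals and interactability per player as quest objectives progress.
; Quest component events set visual and interaction state; objective updates trigger hide-or-swap, effects and cosmetics.
; Discovery overlap and timed distance checks drive unhide/rehide scale timelines and ambient VFX/audio.
; Server-side overlap, vehicle, tag and conversation paths call CheckTagsandComplete, while a test mode bypasses quest requirements and resets after cooldown.
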(event Custom|CheckNextIntervalDistanceMax
  (params (HighestSpeed :type Float%20%28double-precision%29))
  ;[0 1 1] IntervalCheck logic (WIP)
  ;;[1 1 1] All the timer events are defined here. (All IntervalCheck functions can likely eventually be nativized)
  (bind _returnvalue (Utilities|Time|SetTimerbyEvent (EventDispatchers|CreateEvent :!event Custom|IntervalCheckDistanceMax) (IntervalCheck|GetNextIntervalDuration (Variables|Advanced|IntervalCheck|GetWatchedActors_Max) HighestSpeed (Variables|Advanced|IntervalCheck|GetWatchDistance_Max))))
  (Variables|Advanced|IntervalCheck|SetTimerHandle_DistanceMax _returnvalue))

(event Custom|IntervalCheckDistanceMax
  ;[0 1 1] IntervalCheck logic (WIP)
  ;;[1 1 1] All the timer events are defined here. (All IntervalCheck functions can likely eventually be nativized)
  ;;;[1 1 1] Calls each function on a timer
  (IntervalCheck|MaxDistance|UpdateWatchedActorsDistanceMax))

(event Custom|CheckNextIntervalDistanceCosmetic
  (params (HighestSpeed :type Float%20%28double-precision%29))
  ;[0 1 1] IntervalCheck logic (WIP)
  ;;[1 1 1] All the timer events are defined here. (All IntervalCheck functions can likely eventually be nativized)
  (bind _returnvalue (Utilities|Time|SetTimerbyEvent (EventDispatchers|CreateEvent :!event Custom|IntervalCheckDistanceCosmetic) (IntervalCheck|GetNextIntervalDuration (Variables|Advanced|IntervalCheck|GetWatchedActors_Cosmetic) HighestSpeed (Variables|Advanced|IntervalCheck|GetWatchDistance_Cosmetic))))
  (Variables|Advanced|IntervalCheck|SetTimerHandle_DistanceCosmetic _returnvalue))

(event Custom|IntervalCheckDistanceCosmetic
  ;[0 1 1] IntervalCheck logic (WIP)
  ;;[1 1 1] All the timer events are defined here. (All IntervalCheck functions can likely eventually be nativized)
  ;;;[1 1 1] Calls each function on a timer
  (IntervalCheck|CosmeticDistance|UpdateWatchedActorsDistanceCosmetic))

(event Custom|IntervalCheckDistanceDiscovery
  ;[0 1 1] IntervalCheck logic (WIP)
  ;;[1 1 1] All the timer events are defined here. (All IntervalCheck functions can likely eventually be nativized)
  ;;;[1 1 1] Calls each function on a timer
  (IntervalCheck|DiscoveryDistance|UpdateWatchedActorsDistanceDiscovery))

(event Custom|CheckNextIntervalDistanceDiscovery
  (params (HighestSpeed :type Float%20%28double-precision%29))
  ;[0 1 1] IntervalCheck logic (WIP)
  ;;[1 1 1] All the timer events are defined here. (All IntervalCheck functions can likely eventually be nativized)
  (bind _returnvalue (Utilities|Time|SetTimerbyEvent (EventDispatchers|CreateEvent :!event Custom|IntervalCheckDistanceDiscovery) (IntervalCheck|GetNextIntervalDuration (Variables|Advanced|IntervalCheck|GetWatchedActors_Discovery) HighestSpeed (Variables|Advanced|IntervalCheck|GetWatchDistance_Discovery))))
  (Variables|Advanced|IntervalCheck|SetTimerHandle_DistanceDiscovery _returnvalue))

(event Custom|OnEnterDistanceMax
  (params (PlayerController :type Fort%20Player%20Controller%20Object%20Reference) (PlayerPawn :type Fort%20Player%20Pawn%20Athena%20Object%20Reference))
  ;[0 1 1] IntervalCheck logic (WIP)
  ;;[1 1 1] Events designers can use (keep exposed to BP)
  (Development|PrintString "Enter Max"))

(event Custom|EventAddWatchedActor
  (params (PlayerController :type Fort%20Player%20Controller%20Object%20Reference))
  ;[0 1 1] IntervalCheck logic (WIP)
  (IntervalCheck|AddWatchedActor PlayerController))

(event Custom|OnIntervalWithinDistanceMax
  (params (PlayerController :type Fort%20Player%20Controller%20Object%20Reference) (PlayerPawn :type Fort%20Player%20Pawn%20Athena%20Object%20Reference))
  ;[0 1 1] IntervalCheck logic (WIP)
  ;;[1 1 1] Events designers can use (keep exposed to BP)
  (Development|PrintString "Interval Max"))

(event Custom|EventRemoveWatchedActor
  (params (PlayerController :type Fort%20Player%20Controller%20Object%20Reference))
  ;[0 1 1] IntervalCheck logic (WIP)
  (IntervalCheck|RemoveWatchedActor PlayerController))

(event Custom|OnExitDistanceMax
  (params (PlayerController :type Fort%20Player%20Controller%20Object%20Reference) (PlayerPawn :type Fort%20Player%20Pawn%20Athena%20Object%20Reference))
  ;[0 1 1] IntervalCheck logic (WIP)
  ;;[1 1 1] Events designers can use (keep exposed to BP)
  (Development|PrintString "Exit Max"))

(event Custom|OnEnterDistanceCosmetic
  (params (PlayerController :type Fort%20Player%20Controller%20Object%20Reference) (PlayerPawn :type Fort%20Player%20Pawn%20Athena%20Object%20Reference))
  ;[0 1 1] IntervalCheck logic (WIP)
  ;;[1 1 1] Events designers can use (keep exposed to BP)
  (CallFunction|OnEnterCosmetic PlayerController PlayerPawn)
  (Development|PrintString "Entered Cosmetic"))

(event Custom|OnIntervalWithinDistanceCosmetic
  (params (PlayerController :type Fort%20Player%20Controller%20Object%20Reference) (PlayerPawn :type Fort%20Player%20Pawn%20Athena%20Object%20Reference))
  ;[0 1 1] IntervalCheck logic (WIP)
  ;;[1 1 1] Events designers can use (keep exposed to BP)
  (Development|PrintString "Interval Cosmetic"))

(event Custom|OnExitDistanceCosmetic
  (params (PlayerController :type Fort%20Player%20Controller%20Object%20Reference) (PlayerPawn :type Fort%20Player%20Pawn%20Athena%20Object%20Reference))
  ;[0 1 1] IntervalCheck logic (WIP)
  ;;[1 1 1] Events designers can use (keep exposed to BP)
  (CallFunction|OnExitCosmetic PlayerController PlayerPawn)
  (Development|PrintString "Exit Cosmetic"))

(event Custom|OnEnterDistanceDiscovery
  (params (PlayerController :type Fort%20Player%20Controller%20Object%20Reference) (PlayerPawn :type Fort%20Player%20Pawn%20Athena%20Object%20Reference))
  ;[0 1 1] IntervalCheck logic (WIP)
  ;;[1 1 1] Events designers can use (keep exposed to BP)
  (CallFunction|OnEnterDiscovery PlayerController PlayerPawn)
  (Development|PrintString "Entered Discovery"))

(event Custom|OnIntervalWithinDistanceDiscovery
  (params (PlayerController :type Fort%20Player%20Controller%20Object%20Reference) (PlayerPawn :type Fort%20Player%20Pawn%20Athena%20Object%20Reference))
  ;[0 1 1] IntervalCheck logic (WIP)
  ;;[1 1 1] Events designers can use (keep exposed to BP)
  (Development|PrintString "Interval Discovery"))

(event Custom|OnExitDistanceDiscovery
  (params (PlayerController :type Fort%20Player%20Controller%20Object%20Reference) (PlayerPawn :type Fort%20Player%20Pawn%20Athena%20Object%20Reference))
  ;[0 1 1] IntervalCheck logic (WIP)
  ;;[1 1 1] Events designers can use (keep exposed to BP)
  (CallFunction|OnExitDiscovery PlayerController PlayerPawn)
  (Development|PrintString "Exit Discovery"))

(event Custom|OnEnterCosmetic
  (params (PlayerController :type Fort%20Player%20Controller%20Object%20Reference) (PlayerPawn :type Fort%20Player%20Pawn%20Athena%20Object%20Reference))
  ;[0 0 0] Distance toggled cosmetics
  ;;[1 1 1] Activate Ambient Audio & FX if in test state or obj active
  (Networking|SwitchHasAuthority
    (:Remote
      (if (not (and (Variables|Design|QuestSetup|GetbTestVisualsNoQuestRequired) (not (Variables|Advanced|Debug|GetbOnVisualTestingCooldown))))
        (bind _asfort_quest_item_definition_athena (Utilities|Casting|CastToFortQuestItemDefinition_Athena (Class|ConsolidatedQuestComponent|GetQuestItemDefinition (Variables|BuildingGameplayActorPropQuest|GetConsolidatedQuestComponent))))
        (bind (_playercontrollerout _completedquest) (Quest|HasPlayerCompletedQuest PlayerController _asfort_quest_item_definition_athena))
        (if (Math|Boolean|ANDBoolean (not (Variables|Design|GetbUseDiscoverCollisionToUnhide)) (or (and (not _completedquest) (ObjectiveState|IsObjectiveActive PlayerController)) (and (Variables|Design|QuestSetup|GetbTestVisualsNoQuestRequired) (not (Variables|Advanced|Debug|GetbOnVisualTestingCooldown)))) (Interaction|BlueprintCanInteract PlayerPawn "AllInteraction"))
          _))
      (VisualState|VFX&Audio|PreInteractionVFX&Audio PlayerController))))

(event Custom|UpdateVisuals
  (params (PlayerController :type Fort%20Player%20Controller%20Object%20Reference))
  (bind _returnvalue (Game|Player|GetPlayerPawn PlayerController))
  ;[0 0 0] Distance toggled cosmetics
  ;;[1 1 1] Activate Ambient Audio & FX if in test state or obj active
  (Networking|SwitchHasAuthority
    (:Remote
      (Utilities|IsValid PlayerController
        (:Is%20Valid
          (if (Pawn|IsLocalPlayerController PlayerController)
            (Utilities|IsValid _returnvalue
              (:Is%20Valid
                (bind _asfort_player_pawn_athena (Utilities|Casting|CastToFortPlayerPawnAthena _returnvalue))
                (CallFunction|OnEnterCosmetic PlayerController _asfort_player_pawn_athena)))))))))

(event EventBeginPlay
  ;[0 0 0] Initialization
  (Utilities|FlowControl|Sequence
    (:then_0
      ;;[1 1 1] Initialize visual test mode
      (if (Variables|Design|QuestSetup|GetbTestVisualsNoQuestRequired)
        (CallFunction|VisualTestingNoQuestRequiredInit)))
    (:then_1
      ;;[1 1 1] Bind Conversation Events
      (if (Utilities|IsValid (Variables|Components|GetNonPlayerConversationComponent) :!class /Script/Engine.KismetSystemLibrary)
        (CallFunction|BindConversationEvents)))))

(event Custom|VisualTestingNoQuestRequired_Init
  (bind _returnvalue (Fortnite|GetFirstLocalFortPlayerController true))
  ;[0 0 0] Initialization
  ;;[1 1 1] Visual Testing for when we don't want to bother granting ourself the quest
  (if (Variables|Design|GetbUseDiscoverCollisionToUnhide)
    (VisualState|VFX&Audio|NoVFX&Audio _returnvalue)
    (else
      (VisualState|VFX&Audio|PreInteractionVFX&Audio _returnvalue)))
  (if (Variables|Design|QuestSetup|GetbRecheckingTriggersOnSuccess)
    (Utilities|Time|SetTimerbyEvent (delegate Custom|VisualTestingRecheckOverlapping) 0.5 true)
    (else
      (CallFunction|ResetForVisualTesting _returnvalue))))

(event OnQuestNotReady(ConsolidatedQuestComponent)
  (params (PlayerController :type Fort%20Player%20Controller%20Object%20Reference))
  ;[0 0 0] QuestComponent Events
  ;;[1 1 1] OnQuestNotReady(QuestComponent)
  (if (Variables|Design|QuestSetup|GetbTestVisualsNoQuestRequired)
    _
    (elif (Utilities|Array|ContainsItem (Variables|Advanced|Interaction|GetFortPlayerControllersWhoInteracted) PlayerController)
      ;;;[1 1 1] If so, don't hide the actors player interacted with
      (if (not (ObjectiveState|HasCompletedObjective PlayerController))
        (if (Variables|Design|QuestSetup|GetbVisibleToAll)
          (if (ObjectiveState|HasCompletedObjective PlayerController)
            ;[1 1 1] If player completed the quest, show the post-interaction version
            (VisualState|Meshes|PostInteractionVisuals PlayerController)
            (else
              (VisualState|Meshes|PreInteractionVisuals PlayerController)))
          (elif (== (Variables|Design|PostInteraction|GetStateChangeDelay) 0)
            (VisualState|Meshes|HiddenVisuals PlayerController)
            (if (Variables|Advanced|IntervalCheck|GetbUsingDistanceCheck)
              (CallFunction|EventRemoveWatchedActor PlayerController)
              (else
                _))
            (InteractionState|HandleSetCannotInteract PlayerController)
            (else
              _))))
      (else
        _))))

(event Custom|VisualTestingRecheckOverlapping
  ;[0 0 0] Initialization
  ;;[1 1 1] Visual Testing for when we don't want to bother granting ourself the quest
  (CallFunction|UpdateOverlappingActors))

(event Custom|BindConversationEvents
  ;[0 0 0] Initialization
  ;;[1 1 1] ConversationEvents
  (EventDispatchers|BindEventtoOnConversationEnded (Variables|Components|GetNonPlayerConversationComponent) (delegate Custom|ConversationEnded)))

(event Custom|ChangeStateForPlayer
  (params (FPC :type Fort%20Player%20Controller%20Object%20Reference))
  (bind (_preinteraction_mesh_transform_location _preinteraction_mesh_transform_rotation _preinteraction_mesh_transform_scale) (Math|Transform|BreakTransform (Variables|GetPreInteraction%20Mesh%20Transform_Location)))
  ;[0 0 0] State changes & Cosmetics
  ;;[1 1 1] Prop State Change on Interact
  (Variables|Advanced|Debug|SetbOnVisualTestingCooldown true)
  (if (Variables|Advanced|IntervalCheck|GetbUsingDistanceCheck)
    (CallFunction|EventRemoveWatchedActor FPC)
    (else
      _))
  (InteractionState|HandleSetCannotInteract FPC)
  ;;;[1 0 0] This needs to work on just the interacting player in splitscreen
  (VisualState|VFX&Audio|NoVFX&Audio FPC)
  (if (< (Variables|Design|CosmeticEvent|GetCosmeticTimeShift) 0.0)
    (if (Variables|Design|CosmeticEvent|GetbHideDuringCosmeticEvent)
      (VisualState|Meshes|HiddenVisuals FPC)
      (else
        _))
    (VisualState|VFX&Audio|CosmeticVFX&Audio FPC)
    (CallFunction|SpawnCosmeticActor)
    (Utilities|FlowControl|Delay (- (Math|Float|Absolute(Float) (Variables|Design|CosmeticEvent|GetCosmeticTimeShift)) (Variables|Design|PostInteraction|GetStateChangeDelay)))
    (else
      _))
  (Utilities|FlowControl|Delay (Variables|Design|PostInteraction|GetStateChangeDelay))
  ;;;[1 1 1] Hiding or Swapping?
  (if (Variables|Design|PostInteraction|GetbHideAfterInteract)
    ;;;;[1 0 0] What is this? -Macker
    (AddComponent|NODE_AddStaticMeshComponent-0)
    ;;;;[1 0 0] Why are we using 3 timelines for tthe same thing? They are each a component.
    (bind (_direction _scaledowntime) (Timeline|T_UnhideCurve PlayFromStart)
      (:Update
        (Transformation|SetWorldScale3D (Variables|Components|GetSM_Quest_PreInteraction) (Math|Vector|Lerp(Vector) _preinteraction_mesh_transform_scale "0.010000,0.010000,0.010000" _scaledowntime)))
      (:Finished
        (VisualState|Meshes|HiddenVisuals FPC)
        ;;;[1 1 1] Post-Interaction Particles and Sound
        (CallFunction|PlayRewardFeedback FPC)
        (Networking|SwitchHasAuthority
          (:Authority
            ;;;;[1 1 1] Applying a Gameplay Effect to Player
            (CallFunction|ManagePlayerGEs FPC)
            ;;;;[1 1 1] Applying Cosmetic Override to Player
            (if (Variables|Design|QuestSetup|GameplayTags,EffectsandCosmetics|GetbShouldApplyCosmeticOverrides)
              (CallFunction|CosmeticOverridestoPlayer FPC)
              (else
                _))
            (VisualState|VFX&Audio|PostInteractionVFX&Audio FPC)
            (if (>= (Variables|Design|CosmeticEvent|GetCosmeticTimeShift) 0.0)
              (Utilities|FlowControl|Delay (Variables|Design|CosmeticEvent|GetCosmeticTimeShift))
              (VisualState|VFX&Audio|CosmeticVFX&Audio FPC)
              (CallFunction|SpawnCosmeticActor)
              (else
                _))
            (CallFunction|ResetForVisualTesting FPC))
          (:Remote
            _))))
    (else
      (VisualState|Meshes|PostInteractionVisuals FPC))))

(event Custom|ConversationEnded
  ;[0 0 0] Initialization
  ;;[1 1 1] ConversationEvents
  (Variables|Advanced|Interaction|SetbBlockLocalInteractOverride false))

(event OnQuestReady(ConsolidatedQuestComponent)
  (params (PlayerController :type Fort%20Player%20Controller%20Object%20Reference))
  ;[0 0 0] QuestComponent Events
  ;;[1 1 1] OnQuestReady(QuestComponent)
  (if (or (ObjectiveState|IsObjectiveActive PlayerController) (Variables|Design|QuestSetup|GetbTestVisualsNoQuestRequired))
    (if (Variables|Advanced|IntervalCheck|GetbUsingDistanceCheck)
      (CallFunction|EventAddWatchedActor PlayerController)
      (else
        _))
    (if (not (Variables|Design|GetbUseDiscoverCollisionToUnhide))
      (InteractionState|HandleSetCanInteract PlayerController)
      (VisualState|Meshes|PreInteractionVisuals PlayerController)
      (VisualState|VFX&Audio|PreInteractionVFX&Audio PlayerController))
    (elif (Variables|Design|QuestSetup|GetbVisibleToAll)
      (VisualState|Meshes|PostInteractionVisuals PlayerController)
      (else
        (VisualState|Meshes|HiddenVisuals PlayerController)))))

(event Custom|OnExitCosmetic
  (params (PlayerController :type Fort%20Player%20Controller%20Object%20Reference) (PlayerPawn :type Fort%20Player%20Pawn%20Athena%20Object%20Reference))
  ;[0 0 0] Distance toggled cosmetics
  ;;[1 1 1] Disabling Ambient Sound and VFX on moving away
  (Networking|SwitchHasAuthority
    (:Remote
      (VisualState|VFX&Audio|NoVFX&Audio PlayerController))))

(event Custom|LocalVisualsTest
  (params (FPC :type Fort%20Player%20Controller%20Object%20Reference))
  ;[0 0 0] State changes & Cosmetics
  ;;[1 1 1] Prop State Change on Interact
  ;;;[1 1 1] Visual Testing mode
  ;;;;[1 1 1] Visual Testing Only (replicated)
  (if (Utilities|Equal(Object) FPC (Fortnite|GetFirstLocalFortPlayerController true))
    (CallFunction|ChangeStateforPlayer FPC)))

(event Custom|ConversationToChangeState
  (params (PlayerPawn :type Actor%20Object%20Reference))
  ;[1 0.6502 0] Used at least once, but not much
  ;;[1 1 1] ConversationToChangeState
  (Networking|SwitchHasAuthority
    (:Authority
      (if (Variables|Design|GetbConversationToComplete)
        (CallFunction|CheckTagsandComplete PlayerPawn)))))

(event OnUpdateQuest(ConsolidatedQuestComponent)
  (params (PlayerController :type Fort%20Player%20Controller%20Object%20Reference))
  ;[0 0 0] QuestComponent Events
  ;;[1 1 1] [QuestComponent] OnCustomUpdateQuest
  (if (or (ObjectiveState|IsObjectiveActive PlayerController) (Variables|Design|QuestSetup|GetbTestVisualsNoQuestRequired))
    (VisualState|Meshes|PreInteractionVisuals PlayerController)
    (if (Variables|Advanced|IntervalCheck|GetbUsingDistanceCheck)
      (CallFunction|EventAddWatchedActor PlayerController)
      (else
        _))
    (if (not (Variables|Design|GetbUseDiscoverCollisionToUnhide))
      (InteractionState|HandleSetCanInteract PlayerController))))

(event Custom|UpdateOverlappingActors(Collision_DiscoverRadius)
  (params (Other%20Actor :type Actor%20Object%20Reference))
  ;[0 0 0] Discovery
  ;;[1 1 1] Enter Discovery
  (bind _data (Utilities|FlowControl|MergeExecutionwithData Other%20Actor (stale OnComponentBeginOverlap(Collision_DiscoverRadius) OtherActor))
    (:Then
      (bind _data_2 (Utilities|FlowControl|MergeExecutionwithData (stale Custom|OnEnterDiscovery PlayerPawn) _data)
        (:Then
          (CallFunction|HandleComponentOverlapDisoveryRadius _data_2 (Utilities|Struct|MakeFortOverlapData (Variables|Design|QuestSetup|GameplayTags,EffectsandCosmetics|GetRequired%20Player%20Tags) (Variables|Design|QuestSetup|GameplayTags,EffectsandCosmetics|GetProhibited%20Player%20Tags) (Variables|Design|QuestSetup|GetbReHideIfConditionsChange) (Variables|Design|GetbCompleteFromDiscovery) (Variables|Design|QuestSetup|GetbTestVisualsNoQuestRequired) (Variables|Advanced|Debug|GetbOnVisualTestingCooldown))))))))

(event Custom|OnEnterDiscovery
  (params (PlayerController :type Fort%20Player%20Controller%20Object%20Reference) (PlayerPawn :type Fort%20Player%20Pawn%20Athena%20Object%20Reference))
  ;[0 0 0] Discovery
  ;;[1 1 1] Enter Discovery
  (bind _data (Utilities|FlowControl|MergeExecutionwithData PlayerPawn (stale Custom|UpdateOverlappingActors(Collision_DiscoverRadius) _data))
    (:Then
      (CallFunction|HandleComponentOverlapDisoveryRadius _data (Utilities|Struct|MakeFortOverlapData (Variables|Design|QuestSetup|GameplayTags,EffectsandCosmetics|GetRequired%20Player%20Tags) (Variables|Design|QuestSetup|GameplayTags,EffectsandCosmetics|GetProhibited%20Player%20Tags) (Variables|Design|QuestSetup|GetbReHideIfConditionsChange) (Variables|Design|GetbCompleteFromDiscovery) (Variables|Design|QuestSetup|GetbTestVisualsNoQuestRequired) (Variables|Advanced|Debug|GetbOnVisualTestingCooldown))))))

(event EventFxTestOverride
  (params (FortPC :type Fort%20Player%20Controller%20Object%20Reference))
  ;[0 0 0] Discovery
  ;;[1 1 1] Natively driven discovery events
  ;;;[1 1 1] Server needs to track the interact or send stat event for complex custom for "discovery'
  (CallFunction|ChangeStateforPlayer FortPC))

(event Custom|Reset%20for%20Visual%20Testing
  (params (PlayerController :type Fort%20Player%20Controller%20Object%20Reference))
  ;[0 0 0] State changes & Cosmetics
  ;;[1 1 1] Prop State Change on Interact
  ;;;[1 1 1] Visual Testing mode
  ;;;;[1 1 1] Reset to pre-interact state for visual testing mode
  (if (Variables|Design|QuestSetup|GetbTestVisualsNoQuestRequired)
    (Utilities|FlowControl|Delay (Variables|Design|QuestSetup|GetTestingCooldown))
    (Variables|Advanced|Interaction|SetbBlockLocalInteractOverride false)
    (VisualState|VFX&Audio|PreInteractionVFX&Audio PlayerController)
    (VisualState|Meshes|PreInteractionVisuals PlayerController)
    (if (Variables|Advanced|IntervalCheck|GetbUsingDistanceCheck)
      (CallFunction|EventAddWatchedActor PlayerController)
      (else
        _))
    (if (Variables|Design|QuestSetup|GameplayTags,EffectsandCosmetics|GetbShouldApplyCosmeticOverrides)
      (CallFunction|CosmeticOverridestoPlayer PlayerController true)
      (else
        _))
    (InteractionState|HandleSetCanInteract PlayerController)
    (Variables|Advanced|Debug|SetbOnVisualTestingCooldown false)))

(event OnComponentBeginOverlap(Collision_DiscoverRadius)
  (params (OverlappedComponent :type Primitive%20Component%20Object%20Reference) (OtherActor :type Actor%20Object%20Reference) (OtherComp :type Primitive%20Component%20Object%20Reference) (OtherBodyIndex :type Integer) (bFromSweep :type Boolean) (SweepResult :type Hit%20Result%20Structure%20%28by%20ref%29))
  ;[0 0 0] Discovery
  ;;[1 1 1] Enter Discovery
  (bind _data (Utilities|FlowControl|MergeExecutionwithData (stale Custom|UpdateOverlappingActors(Collision_DiscoverRadius) Other%20Actor) OtherActor)
    (:Then
      (bind _data_2 (Utilities|FlowControl|MergeExecutionwithData (stale Custom|OnEnterDiscovery PlayerPawn) _data)
        (:Then
          (CallFunction|HandleComponentOverlapDisoveryRadius _data_2 (Utilities|Struct|MakeFortOverlapData (Variables|Design|QuestSetup|GameplayTags,EffectsandCosmetics|GetRequired%20Player%20Tags) (Variables|Design|QuestSetup|GameplayTags,EffectsandCosmetics|GetProhibited%20Player%20Tags) (Variables|Design|QuestSetup|GetbReHideIfConditionsChange) (Variables|Design|GetbCompleteFromDiscovery) (Variables|Design|QuestSetup|GetbTestVisualsNoQuestRequired) (Variables|Advanced|Debug|GetbOnVisualTestingCooldown))))))))

(event Custom|OnPlayerAction
  (params (PlayerController :type Fort%20Player%20Controller%20Object%20Reference) (ActionTag :type Gameplay%20Tag%20Structure) (PlayerTags :type Gameplay%20Tag%20Container%20Structure))
  (bind _btestvisualsnoquestrequired (Variables|Design|QuestSetup|GetbTestVisualsNoQuestRequired))
  ;[1 0.6502 0] Used at least once, but not much
  ;;[1 1 1] Player Action Tag to Interact
  (if (Variables|Advanced|Interaction|GetbListenToTagToInteract)
    ;;;[1 1 1] Debugging - Display Tags
    (Development|PrintString (Utilities|String|Append "; Player Controller: " (Utilities|GetDisplayName PlayerController) "; Actor Tags: " (GameplayTags|GetDebugStringfromGameplayTagContainer PlayerTags)) _btestvisualsnoquestrequired _btestvisualsnoquestrequired "(R=0.000000,G=0.660000,B=1.000000,A=1.000000)" 3e+01)
    (Development|PrintString (Utilities|String|Append "[Received] Event Tag: " (GameplayTags|GetDebugStringfromGameplayTag ActionTag)) _btestvisualsnoquestrequired _btestvisualsnoquestrequired "(R=1.000000,G=0.084869,B=0.000000,A=1.000000)" 3e+01)
    (Networking|SwitchHasAuthority
      (:Authority
        (if (GameplayTags|HasAnyTags (GameplayTags|MakeGameplayTagContainerfromTag ActionTag) (Variables|Design|QuestSetup|InteractionSettings|GetDamage%20or%20Interact%20Tag))
          (CallFunction|CheckTagsandComplete PlayerController))))))

(event EventHideMesh
  (params (bIsInSplitScreen :type Boolean))
  (bind (_preinteraction_mesh_transform_location _preinteraction_mesh_transform_rotation _preinteraction_mesh_transform_scale) (Math|Transform|BreakTransform (Variables|GetPreInteraction%20Mesh%20Transform_Location)))
  ;[0 0 0] Discovery
  ;;[1 1 1] Natively driven discovery events
  ;;;[1 1 1] Unhiding
  (VisualState|Meshes|PreInteractionVisuals (Variables|BuildingGameplayActorPropQuest|GetFortPlayerController_TempRef))
  ;;;;[1 0 0] Why are we using 3 timelines for tthe same thing? They are each a component.
  (bind (_direction _scaledowntime) (Timeline|T_HideCurve PlayFromStart)
    (:Update
      (Transformation|SetWorldScale3D (Variables|Components|GetSM_Quest_PreInteraction) (Math|Vector|Lerp(Vector) "0.010000,0.010000,0.010000" _preinteraction_mesh_transform_scale _scaledowntime)))))

(event OnQuestComponentSuccessEvent(ConsolidatedQuestComponent)
  (params (PlayerController :type Fort%20Player%20Controller%20Object%20Reference))
  (bind _consolidatedquestcomponent (Variables|BuildingGameplayActorPropQuest|GetConsolidatedQuestComponent))
  (bind _returnvalue (Class|ConsolidatedQuestComponent|GetQuestItemDefinition _consolidatedquestcomponent))
  (bind _returnvalue_2 (Quests|GetQuestManager PlayerController "Invalid"))
  (bind _returnvalue_3 (Quests|GetQuestwithDefinition _returnvalue_2 _returnvalue))
  ;[0 0 0] QuestComponent Events
  ;;[1 1 1] OnQuestComponentSuccessEvent
  (CallFunction|UpdateOverlappingActors)
  (CallFunction|UpdateVisuals PlayerController)
  (Networking|SwitchHasAuthority
    (:Authority
      (Utilities|IsValid _returnvalue_3
        (:Is%20Valid
          (if (Quest|HasCompletedAllObjectives _returnvalue_3 false)
            (if (and (Variables|Design|QuestSetup|GameplayTags,EffectsandCosmetics|GetbShouldApplyCosmeticOverrides) (Variables|Design|QuestSetup|GameplayTags,EffectsandCosmetics|GetbResetOverridesOnCompletion))
              (CallFunction|CosmeticOverridestoPlayer PlayerController true)
              (Development|PrintString "REMOVED" (Variables|Design|QuestSetup|GetbTestVisualsNoQuestRequired) (Variables|Design|QuestSetup|GetbTestVisualsNoQuestRequired) "(R=0.000000,G=1.000000,B=0.073006,A=1.000000)"))
            (else
              (Development|PrintString (Utilities|GetDisplayName _returnvalue)))))))))

(event EventCompleteFromDiscovery
  ;[0 0 0] Discovery
  ;;[1 1 1] Natively driven discovery events
  ;;;[1 1 1] Complete the Quest From Discovery
  (CallFunction|PlayRewardFeedback (Variables|BuildingGameplayActorPropQuest|GetFortPlayerController_TempRef)))

(event EventActivateComponents
  (bind _fortplayercontroller_tempref (Variables|BuildingGameplayActorPropQuest|GetFortPlayerController_TempRef))
  ;[0 0 0] Discovery
  ;;[1 1 1] Natively driven discovery events
  ;;;[1 1 1] Activate Components
  (VisualState|VFX&Audio|UnhideVFX&Audio _fortplayercontroller_tempref)
  (VisualState|VFX&Audio|PreInteractionVFX&Audio _fortplayercontroller_tempref))

(event OnComponentEndOverlap(Collision_DiscoverRadius)
  (params (OverlappedComponent :type Primitive%20Component%20Object%20Reference) (OtherActor :type Actor%20Object%20Reference) (OtherComp :type Primitive%20Component%20Object%20Reference) (OtherBodyIndex :type Integer))
  ;[0 0 0] Discovery
  ;;[1 1 1] Exit Discovery
  (bind _asfort_player_pawn_athena (Utilities|Casting|CastToFortPlayerPawnAthena OtherActor))
  (CallFunction|OnExitDiscovery (Fortnite|GetFortPlayerControllerfromActor _asfort_player_pawn_athena) _asfort_player_pawn_athena))

(event Custom|OnExitDiscovery
  (params (PlayerController :type Fort%20Player%20Controller%20Object%20Reference) (PlayerPawn :type Fort%20Player%20Pawn%20Athena%20Object%20Reference))
  ;[0 0 0] Discovery
  ;;[1 1 1] Exit Discovery
  ;;;[1 0.1414 0] This is gnarly, don't like this - for future optimization (Woj)
  (if (Pawn|IsLocallyControlled PlayerPawn)
    (CallFunction|BPReHideActor PlayerController)
    (else
      _))
  (Utilities|Array|RemoveItem (Variables|BuildingGameplayActorPropQuest|GetOverlapWatchedActors) PlayerPawn))

(event Custom|ForceInteractStateChange
  (params (FPC :type Fort%20Player%20Controller%20Object%20Reference))
  ;[0 0 0] State changes & Cosmetics
  ;;[1 1 1] Force state changes when skipping objective completion (mainly for multi-actor puzzles)
  (if (not (Utilities|Array|ContainsItem (Variables|Advanced|Interaction|GetFortPlayerControllersWhoInteracted) FPC))
    (CallFunction|LocalVisualsTest FPC)
    (CallFunction|ChangeStateforPlayer FPC)
    (Utilities|Array|AddUnique (Variables|Advanced|Interaction|GetFortPlayerControllersWhoInteracted) FPC)))

(event OnCustomSetVisualstoPassed(ConsolidatedQuestComponent)
  (params (PlayerController :type Fort%20Player%20Controller%20Object%20Reference) (bFullyInitialized :type Boolean)))

(event Custom|BP_ReHideActor
  (params (Fort%20PC :type Fort%20Player%20Controller%20Object%20Reference))
  (bind (_value _returnvalue) (Utilities|Map|Find (Variables|BuildingGameplayActorPropQuest|GetFortPlayerControllerToCanInteract) (Variables|BuildingGameplayActorPropQuest|GetFortPlayerController_TempRef)))
  (bind (_preinteraction_mesh_transform_location _preinteraction_mesh_transform_rotation _preinteraction_mesh_transform_scale) (Math|Transform|BreakTransform (Variables|GetPreInteraction%20Mesh%20Transform_Location)))
  (bind _fortplayercontroller_tempref (Variables|BuildingGameplayActorPropQuest|GetFortPlayerController_TempRef))
  ;[0 0 0] Discovery
  ;;[1 1 1] Rehide the actor on exit overlap
  (bind _data (Utilities|FlowControl|MergeExecutionwithData Fort%20PC (stale EventReHideActor FortPC))
    (:Then
      (Variables|BuildingGameplayActorPropQuest|SetFortPlayerController_TempRef _data)
      (if (and _value _returnvalue)
        (VisualState|VFX&Audio|UnhideVFX&Audio (Variables|BuildingGameplayActorPropQuest|GetFortPlayerController_TempRef))
        ;;;[1 0 0] Why are we using 3 timelines for tthe same thing? They are each a component.
        (bind (_direction _scaledowntime) (Timeline|Timeline_0 ReverseFromEnd)
          (:Update
            (Transformation|SetWorldScale3D (Variables|Components|GetSM_Quest_PreInteraction) (Math|Vector|Lerp(Vector) "0.010000,0.010000,0.010000" _preinteraction_mesh_transform_scale _scaledowntime)))
          (:Finished
            (VisualState|Meshes|HiddenVisuals _fortplayercontroller_tempref)
            (VisualState|VFX&Audio|NoVFX&Audio _fortplayercontroller_tempref)
            (InteractionState|HandleSetCannotInteract (Variables|BuildingGameplayActorPropQuest|GetFortPlayerController_TempRef))))))))

(event OnCustomSetVisualstoFailed(ConsolidatedQuestComponent)
  (params (PlayerController :type Fort%20Player%20Controller%20Object%20Reference)))

(event EventReHideActor
  (params (FortPC :type Fort%20Player%20Controller%20Object%20Reference))
  (bind (_value _returnvalue) (Utilities|Map|Find (Variables|BuildingGameplayActorPropQuest|GetFortPlayerControllerToCanInteract) (Variables|BuildingGameplayActorPropQuest|GetFortPlayerController_TempRef)))
  (bind (_preinteraction_mesh_transform_location _preinteraction_mesh_transform_rotation _preinteraction_mesh_transform_scale) (Math|Transform|BreakTransform (Variables|GetPreInteraction%20Mesh%20Transform_Location)))
  (bind _fortplayercontroller_tempref (Variables|BuildingGameplayActorPropQuest|GetFortPlayerController_TempRef))
  ;[0 0 0] Discovery
  ;;[1 1 1] Rehide the actor on exit overlap
  (bind _data (Utilities|FlowControl|MergeExecutionwithData (stale Custom|BP_ReHideActor Fort%20PC) FortPC)
    (:Then
      (Variables|BuildingGameplayActorPropQuest|SetFortPlayerController_TempRef _data)
      (if (and _value _returnvalue)
        (VisualState|VFX&Audio|UnhideVFX&Audio (Variables|BuildingGameplayActorPropQuest|GetFortPlayerController_TempRef))
        ;;;[1 0 0] Why are we using 3 timelines for tthe same thing? They are each a component.
        (Timeline|Timeline_0 ReverseFromEnd)))))

(event OnCustomUpdateObjective(ConsolidatedQuestComponent)
  (params (PlayerController :type Fort%20Player%20Controller%20Object%20Reference) (AchievedCount :type Integer) (RequiredCount :type Integer))
  ;[0 0 0] QuestComponent Events
  ;;[1 1 1] OnCustomUpdateObjective(QuestComponent)
  ;;;[1 1 1] In Testing Mode The Cosmetic Event Was called already
  (if (Variables|Design|QuestSetup|GetbTestVisualsNoQuestRequired)
    _
    (elif (Variables|Design|GetbCompleteFromDiscovery)
      _
      (else
        ;;;[1 1 1] do we depend on the server rep from the quest manager for having the quest completed or assume they did?
        ;;;
        ;;; connect this backup if we want the former
        (CallFunction|ChangeStateforPlayer PlayerController)
        (Utilities|Array|AddUnique (Variables|Advanced|Interaction|GetFortPlayerControllersWhoInteracted) PlayerController)))))

(event OnComponentBeginOverlap(SM_HiddenInteractionCollision)
  (params (OverlappedComponent :type Primitive%20Component%20Object%20Reference) (OtherActor :type Actor%20Object%20Reference) (OtherComp :type Primitive%20Component%20Object%20Reference) (OtherBodyIndex :type Integer) (bFromSweep :type Boolean) (SweepResult :type Hit%20Result%20Structure%20%28by%20ref%29))
  ;[0 0 0] Interaction Verbs (Non-discovery)
  ;;[1 1 1] Overlap to Change State - Listen For Tag, Vehicles etc.
  (bind _data (Utilities|FlowControl|MergeExecutionwithData OtherActor (stale Custom|UpdateOverlappingActors(SM_HiddenInteractionCollision) OtherActor))
    (:Then
      (Networking|SwitchHasAuthority
        (:Authority
          (if (Variables|Advanced|Interaction|GetbOverlapToInteract)
            (if (Variables|Advanced|Interaction|GetbListenToTagToInteract)
              (bind _asfort_player_pawn (Utilities|Casting|CastToFortPlayerPawn _data))
              (if (ObjectiveState|IsObjectiveActive (Fortnite|GetFortPlayerControllerfromActor _asfort_player_pawn))
                _
                (elif (Variables|Design|QuestSetup|GetbTestVisualsNoQuestRequired)
                  _))
              (GameplayEffects|ApplyGameplayEffectToTarget :self (Ability|GetAbilitySystemComponent self) :GameplayEffectClass (Variables|Advanced|Interaction|GetTagListenerEffect) :Target (Ability|GetAbilitySystemComponent _asfort_player_pawn) :Level 0.0)
              (CallFunction|PlayerinInteractRange _asfort_player_pawn)
              (else
                (switch Utilities|FlowControl|Switch|SwitchonE_Athena_GenericQuestDamageTypes (Variables|Design|QuestSetup|InteractionSettings|GetInteractionType)
                  (:NewEnumerator9
                    ;;;[1 1 1] Overlap  to Interact
                    (bind _asfort_vehicle_interface (Utilities|Casting|CastToFortVehicleInterface _data))
                    (for _array_element (Vehicle|GetAllPassengers _asfort_vehicle_interface)
                      (CallFunction|CheckTagsandComplete _array_element)))
                  (:NewEnumerator18
                    (bind _asfort_physics_ball (Utilities|Casting|CastToFortPhysicsBall _data))
                    (CallFunction|CheckTagsandComplete (Actor|GetOwner _asfort_physics_ball)))
                  (:NewEnumerator20
                    (bind _asfort_player_pawn_2 (Utilities|Casting|CastToFortPlayerPawn _data))
                    (CallFunction|CheckTagsandComplete _asfort_player_pawn_2))
                  (:NewEnumerator21
                    (bind _asfort_vehicle_interface_2 (Utilities|Casting|CastToFortVehicleInterface _data)
                      (:then
                        (for _array_element_2 (Vehicle|GetAllPassengers _asfort_vehicle_interface_2)
                          (CallFunction|CheckTagsandComplete _array_element_2)))
                      (:CastFailed
                        (bind _asfort_player_pawn_3 (Utilities|Casting|CastToFortPlayerPawn _data))
                        (CallFunction|CheckTagsandComplete _asfort_player_pawn_3)))))))))))))

(event Custom|UpdateOverlappingActors(SM_HiddenInteractionCollision)
  (params (OtherActor :type Actor%20Object%20Reference))
  ;[0 0 0] Interaction Verbs (Non-discovery)
  ;;[1 1 1] Overlap to Change State - Listen For Tag, Vehicles etc.
  (bind _data (Utilities|FlowControl|MergeExecutionwithData (stale OnComponentBeginOverlap(SM_HiddenInteractionCollision) OtherActor) OtherActor)
    (:Then
      (Networking|SwitchHasAuthority
        (:Authority
          (if (Variables|Advanced|Interaction|GetbOverlapToInteract)
            (if (Variables|Advanced|Interaction|GetbListenToTagToInteract)
              (bind _asfort_player_pawn (Utilities|Casting|CastToFortPlayerPawn _data))
              (if (ObjectiveState|IsObjectiveActive (Fortnite|GetFortPlayerControllerfromActor _asfort_player_pawn))
                _
                (elif (Variables|Design|QuestSetup|GetbTestVisualsNoQuestRequired)
                  _))
              (GameplayEffects|ApplyGameplayEffectToTarget :self (Ability|GetAbilitySystemComponent self) :GameplayEffectClass (Variables|Advanced|Interaction|GetTagListenerEffect) :Target (Ability|GetAbilitySystemComponent _asfort_player_pawn) :Level 0.0)
              (CallFunction|PlayerinInteractRange _asfort_player_pawn)
              (else
                (switch Utilities|FlowControl|Switch|SwitchonE_Athena_GenericQuestDamageTypes (Variables|Design|QuestSetup|InteractionSettings|GetInteractionType)
                  (:NewEnumerator9
                    ;;;[1 1 1] Overlap  to Interact
                    (bind _asfort_vehicle_interface (Utilities|Casting|CastToFortVehicleInterface _data))
                    (for _array_element (Vehicle|GetAllPassengers _asfort_vehicle_interface)
                      (CallFunction|CheckTagsandComplete _array_element)))
                  (:NewEnumerator18
                    (bind _asfort_physics_ball (Utilities|Casting|CastToFortPhysicsBall _data))
                    (CallFunction|CheckTagsandComplete (Actor|GetOwner _asfort_physics_ball)))
                  (:NewEnumerator20
                    (bind _asfort_player_pawn_2 (Utilities|Casting|CastToFortPlayerPawn _data))
                    (CallFunction|CheckTagsandComplete _asfort_player_pawn_2))
                  (:NewEnumerator21
                    (bind _asfort_vehicle_interface_2 (Utilities|Casting|CastToFortVehicleInterface _data)
                      (:then
                        (for _array_element_2 (Vehicle|GetAllPassengers _asfort_vehicle_interface_2)
                          (CallFunction|CheckTagsandComplete _array_element_2)))
                      (:CastFailed
                        (bind _asfort_player_pawn_3 (Utilities|Casting|CastToFortPlayerPawn _data))
                        (CallFunction|CheckTagsandComplete _asfort_player_pawn_3)))))))))))))

(event Custom|OnEnterInteract)

(event Custom|PlayerInInteractRange
  (params (Actor :type Actor%20Object%20Reference)))

(event OnComponentEndOverlap(SM_HiddenInteractionCollision)
  (params (OverlappedComponent :type Primitive%20Component%20Object%20Reference) (OtherActor :type Actor%20Object%20Reference) (OtherComp :type Primitive%20Component%20Object%20Reference) (OtherBodyIndex :type Integer))
  ;[0 0 0] Interaction Verbs (Non-discovery)
  ;;[1 1 1] Overlap to Change State - Listen For Tag, Vehicles etc.
  ;;;[1 1 1] End Overlap with interaction collision
  (Networking|SwitchHasAuthority
    (:Authority
      (if (and (Variables|Advanced|Interaction|GetbOverlapToInteract) (Variables|Advanced|Interaction|GetbListenToTagToInteract))
        (bind _asfort_player_pawn (Utilities|Casting|CastToFortPlayerPawn OtherActor))
        (GameplayEffects|RemoveActiveGameplayEffectbySourceEffect (Ability|GetAbilitySystemComponent _asfort_player_pawn) (Variables|Advanced|Interaction|GetTagListenerEffect) (Ability|GetAbilitySystemComponent self) -1)
        (CallFunction|PlayerOutOfInteractRange _asfort_player_pawn)))))

(event Custom|PlayerOutOfInteractRange
  (params (Actor :type Actor%20Object%20Reference)))

(event Custom|OnExitInteract)

(event EventOnDamageServer
  (params (Damage :type Float%20%28single-precision%29) (DamageTags :type Gameplay%20Tag%20Container%20Structure%20%28by%20ref%29) (Momentum :type Vector :default "0, 0, 0") (HitInfo :type Hit%20Result%20Structure%20%28by%20ref%29) (InstigatedBy :type Controller%20Object%20Reference) (DamageCauser :type Actor%20Object%20Reference) (EffectContext :type Gameplay%20Effect%20Context%20Handle%20Structure))
  ;[0 0 0] Interaction Verbs (Non-discovery)
  ;;[1 1 1] Damage to Change State
  (if (Variables|Advanced|Interaction|GetbDamageToInteract)
    (if (GameplayTags|HasAnyTags DamageTags (Variables|Design|QuestSetup|InteractionSettings|GetDamage%20or%20Interact%20Tag))
      (bind _asfort_player_controller (Utilities|Casting|CastToFortPlayerController (Utilities|SelectObject InstigatedBy (Ability|EffectContext|GetEffectCauser EffectContext) (Utilities|IsValid InstigatedBy :!class /Script/Engine.KismetSystemLibrary))))
      (CallFunction|CheckTagsandComplete _asfort_player_controller))))

(fn AddWatchedActor
  (params (PlayerController :type Fort%20Player%20Controller%20Object%20Reference))
  (if (Variables|Advanced|IntervalCheck|GetbUsingDistanceCheck)
    (bind _asfort_player_pawn_athena (Utilities|Casting|CastToFortPlayerPawnAthena (Game|Player|GetPlayerPawn PlayerController)))
    (IntervalCheck|MaxDistance|HandleEnterDistanceMax _asfort_player_pawn_athena)
    (if (not (Utilities|Time|IsTimerActivebyHandle (Variables|Advanced|IntervalCheck|GetTimerHandle_DistanceMax)))
      (CallFunction|IntervalCheckDistanceMax))))

(fn BlueprintCanInteract
  (params (InteractingPawn :type Fort%20Pawn%20Object%20Reference) (InteractionBeingAttempted :type EInteractionBeingAttempted%20Enum :default "FirstInteraction") (InteractionType :type TInteractionType%20Enum :default "IT_NoInteraction"))
  (bind _returnvalue (GameplayTags|GetOwnedGameplayTags InteractingPawn :!class /Script/GameplayTags.GameplayTagAssetInterface))
  (bind _prohibited_player_tags (Variables|Design|QuestSetup|GameplayTags,EffectsandCosmetics|GetProhibited%20Player%20Tags))
  (bind _returnvalue_2 (GameplayTags|HasAnyTags _returnvalue _prohibited_player_tags true))
  (bind _returnvalue_3 (not _returnvalue_2))
  (bind _required_player_tags (Variables|Design|QuestSetup|GameplayTags,EffectsandCosmetics|GetRequired%20Player%20Tags))
  (bind _returnvalue_4 (GameplayTags|HasAllTags _returnvalue _required_player_tags true))
  (bind _bonvisualtestingcooldown (Variables|Advanced|Debug|GetbOnVisualTestingCooldown))
  (bind _returnvalue_5 (not _bonvisualtestingcooldown))
  (bind _bdeposittointeract (Variables|Advanced|Interaction|GetbDepositToInteract))
  (bind _returnvalue_6 (not _bdeposittointeract))
  (bind _returnvalue_7 (and _returnvalue_3 _returnvalue_4))
  (bind _resourceitemdefinition (Variables|Advanced|Interaction|GetResourceItemDefinition))
  (bind _btestvisualsnoquestrequired (Variables|Design|QuestSetup|GetbTestVisualsNoQuestRequired))
  (bind _depositamount (Variables|Design|QuestSetup|InteractionSettings|GetDepositAmount))
  (bind _asfort_player_controller (Utilities|Casting|CastToFortPlayerController (Pawn|GetController InteractingPawn))
    (:then
      (if (Math|Boolean|ANDBoolean _btestvisualsnoquestrequired _returnvalue_5 _returnvalue_7 (or _returnvalue_6 (>= (Inventory|GetItemQuantityOnPlayer _asfort_player_controller _resourceitemdefinition) _depositamount)))
        (return true)
        (else
          (LocalVariables|SetLocal_FortPlayerControllerAthena _asfort_player_controller)
          (bind _asfort_quest_item_definition_athena (Utilities|Casting|CastToFortQuestItemDefinition_Athena (Class|ConsolidatedQuestComponent|GetQuestItemDefinition (Variables|BuildingGameplayActorPropQuest|GetConsolidatedQuestComponent)))
            (:then
              (bind (_playercontrollerout _completedquest) (Quest|HasPlayerCompletedQuest _asfort_player_controller _asfort_quest_item_definition_athena)))
            (:CastFailed
              _))
          (if (not (and _completedquest (not (Variables|Design|QuestSetup|GetbTestVisualsNoQuestRequired))))
            (Networking|SwitchHasAuthority
              (:Authority
                (return (Math|Boolean|ANDBoolean (or (ObjectiveState|IsObjectiveActive _asfort_player_controller) (Variables|Design|QuestSetup|GetbTestVisualsNoQuestRequired)) _returnvalue_7 (or _returnvalue_6 (>= (Inventory|GetItemQuantityOnPlayer _asfort_player_controller _resourceitemdefinition) _depositamount)))))
              (:Remote
                (if (Variables|Advanced|Interaction|GetbBlockLocalInteractOverride)
                  _
                  (elif (Variables|Design|GetbConversationToComplete)
                    (bind _returnvalue_8 (Conversation|FindConversationComponent _asfort_player_controller))
                    (Utilities|IsValid _returnvalue_8
                      (:Is%20Valid
                        (if (not (Conversation|IsinActiveConversation _returnvalue_8))
                          (return (or (Math|Boolean|ANDBoolean (Utilities|Map|Find (Variables|BuildingGameplayActorPropQuest|GetFortPlayerControllerToCanInteract) _asfort_player_controller) (.ReturnValue (Utilities|Map|Find (Variables|BuildingGameplayActorPropQuest|GetFortPlayerControllerToCanInteract) _asfort_player_controller)) _returnvalue_7 (or _returnvalue_6 (>= (Inventory|GetItemQuantityOnPlayer _asfort_player_controller _resourceitemdefinition) _depositamount))) (Math|Boolean|ANDBoolean _btestvisualsnoquestrequired _returnvalue_5 _returnvalue_7 (or _returnvalue_6 (>= (Inventory|GetItemQuantityOnPlayer _asfort_player_controller _resourceitemdefinition) _depositamount)))))
                          (else
                            _)))
                      (:Is%20Not%20Valid
                        _))
                    (return false)
                    (else
                      _))))))
          (return false))))
    (:CastFailed
      (return false))))

(fn BlueprintGetFailedInteractionString
  (params (InteractingPawn :type Fort%20Pawn%20Object%20Reference) (InteractionBeingAttempted :type EInteractionBeingAttempted%20Enum :default "FirstInteraction"))
  (bind _returnvalue (Fortnite|GetFortPlayerControllerfromActor InteractingPawn))
  (bind _amountneeded (LocalVariables|GetAmountNeeded))
  (if (Variables|Advanced|Interaction|GetbDepositToInteract)
    (if (or (ObjectiveState|IsObjectiveActive _returnvalue) (and (Variables|Design|QuestSetup|GetbTestVisualsNoQuestRequired) (not (Variables|Advanced|Debug|GetbOnVisualTestingCooldown))))
      (bind _output_get (LocalVariables|SetAmountNeeded (- (Variables|Design|QuestSetup|InteractionSettings|GetDepositAmount) (Inventory|GetItemQuantityOnPlayer _returnvalue (Variables|Advanced|Interaction|GetResourceItemDefinition)))))
      (if (> _output_get 0)
        (switch Utilities|FlowControl|Switch|SwitchonE_Athena_GenericQuestDamageTypes (Variables|Design|QuestSetup|InteractionSettings|GetInteractionType)
          (:NewEnumerator13
            (return (Utilities|Text|FormatText "NSLOCTEXT(\"[555A9C9E42F1CA7C0B294C3C3484A8EB]\", \"7476CE424BC57963EBCDE08BF50D069C\", \"Need {amount} more bars\")" _amountneeded)))
          (:NewEnumerator14
            (return (Utilities|Text|FormatText "NSLOCTEXT(\"[555A9C9E42F1CA7C0B294C3C3484A8EB]\", \"267397C24AD23B526A8D84B421AF39DA\", \"Need {amount} more wood\")" _amountneeded)))
          (:NewEnumerator15
            (return (Utilities|Text|FormatText "NSLOCTEXT(\"[555A9C9E42F1CA7C0B294C3C3484A8EB]\", \"68CB21E94CE5E20F87D0589C6F466A9B\", \"Need {amount} more stone\")" _amountneeded)))
          (:NewEnumerator16
            (return (Utilities|Text|FormatText "NSLOCTEXT(\"[555A9C9E42F1CA7C0B294C3C3484A8EB]\", \"52458AA94AE47DBEB1B9D2AF2D255CA1\", \"Need {amount} more metal\")" _amountneeded))))
        (else
          (return "")))
      (else
        _))
    (else
      _))
  (return ""))

(fn BlueprintGetInteractionString
  (params (InteractingPawn :type Fort%20Pawn%20Object%20Reference) (InteractionBeingAttempted :type EInteractionBeingAttempted%20Enum :default "FirstInteraction"))
  (bind _depositamount (Variables|Design|QuestSetup|InteractionSettings|GetDepositAmount))
  (if (Variables|Advanced|Interaction|GetbDepositToInteract)
    ;[1 1 1] Fixing Loc Issue with dynamic strings
    (switch Utilities|FlowControl|Switch|SwitchonE_Athena_GenericQuestDamageTypes (Variables|Design|QuestSetup|InteractionSettings|GetInteractionType)
      (:NewEnumerator13
        (return (Utilities|Text|ToText(String) (Utilities|String|Append (Utilities|String|ToString(Text) (Variables|Design|QuestSetup|InteractionSettings|GetInteraction%20Verb)) " " (Utilities|String|ToString(Text) (Utilities|Text|FormatText "NSLOCTEXT(\"[555A9C9E42F1CA7C0B294C3C3484A8EB]\", \"603D073B4C0BBA6CAEBF52BC17288613\", \"{amount} bars\")" _depositamount))))))
      (:NewEnumerator14
        (return (Utilities|Text|ToText(String) (Utilities|String|Append (Utilities|String|ToString(Text) (Variables|Design|QuestSetup|InteractionSettings|GetInteraction%20Verb)) " " (Utilities|String|ToString(Text) (Utilities|Text|FormatText "NSLOCTEXT(\"[555A9C9E42F1CA7C0B294C3C3484A8EB]\", \"8049B1994536E33EB4F54C9CA84E15D7\", \"{amount} wood\")" _depositamount))))))
      (:NewEnumerator15
        (return (Utilities|Text|ToText(String) (Utilities|String|Append (Utilities|String|ToString(Text) (Variables|Design|QuestSetup|InteractionSettings|GetInteraction%20Verb)) " " (Utilities|String|ToString(Text) (Utilities|Text|FormatText "NSLOCTEXT(\"[555A9C9E42F1CA7C0B294C3C3484A8EB]\", \"AD8D9F5E42C992AC12F35D98E2017A0E\", \"{amount} stone\")" _depositamount))))))
      (:NewEnumerator16
        (return (Utilities|Text|ToText(String) (Utilities|String|Append (Utilities|String|ToString(Text) (Variables|Design|QuestSetup|InteractionSettings|GetInteraction%20Verb)) " " (Utilities|String|ToString(Text) (Utilities|Text|FormatText "NSLOCTEXT(\"[555A9C9E42F1CA7C0B294C3C3484A8EB]\", \"70C30E82408EE445C6FA8C8F3DDEE433\", \"{amount} metal\")" _depositamount)))))))
    (else
      (return (Variables|Design|QuestSetup|InteractionSettings|GetInteraction%20Verb)))))

(fn BlueprintGetInteractionTime
  (params (InteractingPawn :type Fort%20Pawn%20Object%20Reference) (InteractionBeingAttempted :type EInteractionBeingAttempted%20Enum :default "FirstInteraction"))
  (Utilities|IsValid (Variables|Components|GetNonPlayerConversationComponent)
    (:Is%20Valid
      (return 0.0 true))
    (:Is%20Not%20Valid
      (return 0.0 false))))

(fn BlueprintOnInteract
  (params (InteractingPawn :type Fort%20Pawn%20Object%20Reference) (InteractionBeingAttempted :type EInteractionBeingAttempted%20Enum :default "FirstInteraction"))
  (bind _interactingpawn (Variables|GetInteractingPawn))
  (bind _returnvalue (Fortnite|GetFortPlayerControllerfromActor _interactingpawn))
  (bind _returnvalue_2 (Fortnite|GetFortPlayerControllerfromActor InteractingPawn))
  (bind _resourceitemdefinition (Variables|Advanced|Interaction|GetResourceItemDefinition))
  (bind _bconversationtocomplete (Variables|Design|GetbConversationToComplete))
  (Utilities|IsValid (Variables|Components|GetNonPlayerConversationComponent)
    (:Is%20Valid
      (Utilities|IsValid _returnvalue_2
        (:Is%20Valid
          (if (Conversation|CanStartConversationWith (Variables|Components|GetNonPlayerConversationComponent) _returnvalue_2)
            (Conversation|StartConversation (Variables|Components|GetNonPlayerConversationComponent) (Utilities|Array|Get(acopy) (GameplayTags|BreakGameplayTagContainer (Variables|Design|QuestSetup|InteractionSettings|GetDamage%20or%20Interact%20Tag))) _returnvalue_2 self :!class /Script/FortniteGame.FortNonPlayerConversationParticipantComponent)
            (else
              _))
          (Development|PrintString "OnServerInteract")
          (Development|PrintString (Utilities|String|Append "Conversation to Complete:
" (Utilities|String|ToString(Boolean) _bconversationtocomplete)))
          (if _bconversationtocomplete
            _
            (elif (Variables|Design|QuestSetup|GetbTestVisualsNoQuestRequired)
              (CallFunction|ChangeStateforPlayer _returnvalue)
              (return)
              (elif (Variables|Advanced|Interaction|GetbDepositToInteract)
                ;[1 1 1] Deposit to Interact logic
                (if (>= (Inventory|GetItemQuantityOnPlayer _returnvalue _resourceitemdefinition) (Variables|Design|QuestSetup|InteractionSettings|GetDepositAmount))
                  (bind _returnvalue_3 (InventoryManagement|FindExistingItemforDefinition _returnvalue _resourceitemdefinition))
                  (InventoryManagement|RemoveItem _returnvalue (Item|GetItemGuid _returnvalue_3) (Variables|Design|QuestSetup|InteractionSettings|GetDepositAmount))
                  (return))
                (elif (Variables|Design|QuestSetup|GetbSkipQuestObjectiveCompletion)
                  ;[1 1 1] Interact logic triggers state change independent of quest completion
                  (CallFunction|ForceInteractStateChange _returnvalue))))))))
    (:Is%20Not%20Valid
      _)))

(fn BlueprintOnLocalInteract
  (params (InteractingPawn :type Fort%20Player%20Pawn%20Object%20Reference) (InteractionBeingAttempted :type EInteractionBeingAttempted%20Enum :default "FirstInteraction"))
  (bind _returnvalue (Fortnite|GetFortPlayerControllerfromActor InteractingPawn))
  (Development|PrintString "BlueprintOnLocalInteract")
  (if (not (Variables|Design|GetbConversationToComplete))
    (Variables|Advanced|Interaction|SetbBlockLocalInteractOverride true)
    (if (Variables|Design|QuestSetup|GetbTestVisualsNoQuestRequired)
      (CallFunction|ChangeStateforPlayer _returnvalue)
      (return false)
      (elif (Variables|Design|QuestSetup|GetbSkipQuestObjectiveCompletion)
        ;[1 1 1] Interact logic triggers state change independent of quest completion
        (CallFunction|ForceInteractStateChange _returnvalue)))))

(fn CheckTagsAndComplete
  (params (Actor :type Actor%20Object%20Reference))
  (bind _returnvalue (Fortnite|GetFortPlayerControllerfromActor Actor))
  (bind _returnvalue_2 (CallFunction|GetQuestTargetTags))
  (bind _returnvalue_3 (CallFunction|PlayerTagsCheck _returnvalue (Variables|Design|QuestSetup|GameplayTags,EffectsandCosmetics|GetRequired%20Player%20Tags) (Variables|Design|QuestSetup|GameplayTags,EffectsandCosmetics|GetProhibited%20Player%20Tags)))
  (if _returnvalue_3
    (if (Variables|Design|QuestSetup|GetbTestVisualsNoQuestRequired)
      (if (not (Variables|Advanced|Debug|GetbOnVisualTestingCooldown))
        (CallFunction|LocalVisualsTest _returnvalue)
        (CallFunction|ChangeStateforPlayer _returnvalue))
      (elif (ObjectiveState|IsObjectiveActive _returnvalue)
        (if (Variables|Design|QuestSetup|GetbSkipQuestObjectiveCompletion)
          (CallFunction|ForceInteractStateChange _returnvalue)
          (elif (Class|FortQuestItemDefinition|HasComponentforWorldProcessingUpdates (Class|ConsolidatedQuestComponent|GetQuestItemDefinition (Variables|BuildingGameplayActorPropQuest|GetConsolidatedQuestComponent)))
            (if (Variables|Design|GetbConversationToComplete)
              (bind _result (CustomVerb|AddTagsforSubject (CustomVerb|MakeCustomVerbMessage _returnvalue "(TagName=\"CustomVerb.ConversationComplete\")") "Prop" (GameplayTags|GetOwnedGameplayTags :!class /Script/GameplayTags.GameplayTagAssetInterface)))
              (CustomVerb|BroadcastCustomVerbMessage _result)
              (else
                (CallFunction|SendInteractionMessage _returnvalue _returnvalue_2)))
            (else
              (Quest|SendComplexCustomStatEvent (Quests|GetQuestManager _returnvalue "Invalid") self "(GameplayTags=)" _returnvalue_2 1))))))))

(fn ConversationMessage
  (params (MessageNumber :type Integer)))

(fn ConversationSuccess
  (params (PlayerController :type Fort%20Player%20Controller%20Object%20Reference))
  (bind _returnvalue (Pawn|GetControlledPawn PlayerController))
  (Utilities|IsValid PlayerController
    (:Is%20Valid
      (Utilities|IsValid _returnvalue
        (:Is%20Valid
          (CallFunction|ConversationtoChangeState _returnvalue))))))

(fn CosmeticOverridestoPlayer
  (params (FPC :type Fort%20Player%20Controller%20Object%20Reference) (Remove :type Boolean))
  (bind _returnvalue (Game|Player|GetPlayerPawn FPC))
  (bind _returnvalue_2 (GUID|GenerateGuidfromObject (Class|ConsolidatedQuestComponent|GetQuestItemDefinition (Variables|BuildingGameplayActorPropQuest|GetConsolidatedQuestComponent))))
  (if Remove
    (Class|FortPlayerPawn|SwapPawnCosmeticItemandVariant _returnvalue _returnvalue_2 (Variables|Design|QuestSetup|GameplayTags,EffectsandCosmetics|GetPlayer%20Cosmetic%20Overrides) false "Inactive" false)
    (Development|PrintString "REMOVED")
    (else
      (Class|FortPlayerPawn|SwapPawnCosmeticItemandVariant _returnvalue _returnvalue_2 (Variables|Design|QuestSetup|GameplayTags,EffectsandCosmetics|GetPlayer%20Cosmetic%20Overrides) false "Active" false))))

(fn CosmeticVFX&Audio
  (params (PlayerController :type Fort%20Player%20Controller%20Object%20Reference))
  (Utilities|IsValid PlayerController
    (:Is%20Valid
      (if (Pawn|IsLocalPlayerController PlayerController)
        (Utilities|FlowControl|Sequence
          (:then_0
            (Components|Activation|Activate (Variables|Components|GetP_Cosmetic) true))
          (:then_1
            (bind _cosmeticaudio (Variables|Components|GetCosmeticAudio)
              (:then
                (Audio|Components|Audio|FadeIn _cosmeticaudio (Variables|Design|CosmeticEvent|GetCosmeticAudioFadeIn) 1.0 0.0 "Linear" :!class /Script/Engine.AudioComponent))
              (:else))))))))

(fn GetNextIntervalDuration
  (params (TargetSet :type Set%20of%20Fort%20Player%20Pawn%20Athena%20Object%20References) (HighestSpeed :type Float%20%28double-precision%29) (Distance :type Float%20%28double-precision%29))
  (if (Utilities|Set|IsEmpty TargetSet)
    (return -1.0)
    (else
      (return (Math|Float|SafeDivide Distance (* (Math|Float|Clamp(Float) HighestSpeed 5.5e+02 4e+03) (Variables|Advanced|IntervalCheck|GetMinChecksBeforeReaching)))))))

(fn HandleEnterDistanceCosmetic
  (params (PlayerPawn :type Fort%20Player%20Pawn%20Athena%20Object%20Reference))
  (if (not (Utilities|Set|ContainsItem (Variables|Advanced|IntervalCheck|GetWatchedActors_Cosmetic) PlayerPawn))
    (Utilities|Set|Add (Variables|Advanced|IntervalCheck|GetWatchedActors_Cosmetic) PlayerPawn)
    (CallFunction|OnEnterDistanceCosmetic (Fortnite|GetFortPlayerControllerfromActor PlayerPawn) PlayerPawn))
  (IntervalCheck|MaxDistance|HandleEnterDistanceMax PlayerPawn))

(fn HandleEnterDistanceDiscovery
  (params (PlayerPawn :type Fort%20Player%20Pawn%20Athena%20Object%20Reference))
  (if (not (Utilities|Set|ContainsItem (Variables|Advanced|IntervalCheck|GetWatchedActors_Discovery) PlayerPawn))
    (Utilities|Set|Add (Variables|Advanced|IntervalCheck|GetWatchedActors_Discovery) PlayerPawn)
    (CallFunction|OnEnterDistanceDiscovery (Fortnite|GetFortPlayerControllerfromActor PlayerPawn) PlayerPawn))
  (IntervalCheck|CosmeticDistance|HandleEnterDistanceCosmetic PlayerPawn))

(fn HandleEnterDistanceMax
  (params (PlayerPawn :type Fort%20Player%20Pawn%20Athena%20Object%20Reference))
  (if (not (Utilities|Set|ContainsItem (Variables|Advanced|IntervalCheck|GetWatchedActors_Max) PlayerPawn))
    (Utilities|Set|Add (Variables|Advanced|IntervalCheck|GetWatchedActors_Max) PlayerPawn)
    (CallFunction|OnEnterDistanceMax (Fortnite|GetFortPlayerControllerfromActor PlayerPawn) PlayerPawn)))

(fn HandleExitDistanceCosmetic
  (params (PlayerPawn :type Fort%20Player%20Pawn%20Athena%20Object%20Reference))
  (bind _returnvalue (Utilities|Set|Remove (Variables|Advanced|IntervalCheck|GetWatchedActors_Cosmetic) PlayerPawn))
  (if _returnvalue
    (CallFunction|OnExitDistanceCosmetic (Fortnite|GetFortPlayerControllerfromActor PlayerPawn) PlayerPawn)
    (else
      _))
  (IntervalCheck|DiscoveryDistance|HandleExitDistanceDiscovery PlayerPawn))

(fn HandleExitDistanceDiscovery
  (params (PlayerPawn :type Fort%20Player%20Pawn%20Athena%20Object%20Reference))
  (bind _returnvalue (Utilities|Set|Remove (Variables|Advanced|IntervalCheck|GetWatchedActors_Discovery) PlayerPawn))
  (if _returnvalue
    (CallFunction|OnExitDistanceDiscovery (Fortnite|GetFortPlayerControllerfromActor PlayerPawn) PlayerPawn)))

(fn HandleExitDistanceMax
  (params (PlayerPawn :type Fort%20Player%20Pawn%20Athena%20Object%20Reference))
  (bind _returnvalue (Utilities|Set|Remove (Variables|Advanced|IntervalCheck|GetWatchedActors_Max) PlayerPawn))
  (if _returnvalue
    (CallFunction|OnExitDistanceMax (Fortnite|GetFortPlayerControllerfromActor PlayerPawn) PlayerPawn)
    (else
      _))
  (IntervalCheck|CosmeticDistance|HandleExitDistanceCosmetic PlayerPawn))

(fn HandleSetCanInteract
  (params (PlayerController :type Fort%20Player%20Controller%20Object%20Reference%20%28by%20ref%29))
  (Utilities|Map|Add (Variables|BuildingGameplayActorPropQuest|GetFortPlayerControllerToCanInteract) PlayerController "true")
  (if (Variables|Advanced|Interaction|GetbDamageToInteract)
    (Networking|SwitchHasAuthority
      (:Remote
        (Collision|SetCollisionEnabled (Variables|Components|GetSM_HiddenInteractionCollision) "QueryOnly")))))

(fn HandleSetCannotInteract
  (params (PlayerController :type Fort%20Player%20Controller%20Object%20Reference%20%28by%20ref%29))
  (Utilities|Map|Remove (Variables|BuildingGameplayActorPropQuest|GetFortPlayerControllerToCanInteract) PlayerController)
  (if (Variables|Advanced|Interaction|GetbDamageToInteract)
    (Networking|SwitchHasAuthority
      (:Remote
        (Collision|SetCollisionEnabled (Variables|Components|GetSM_HiddenInteractionCollision) "NoCollision")))))

(fn HandleWithinDistanceCosmetic
  (params (PlayerPawn :type Fort%20Player%20Pawn%20Athena%20Object%20Reference))
  (IntervalCheck|CosmeticDistance|HandleEnterDistanceCosmetic PlayerPawn)
  (IntervalCheck|DiscoveryDistance|HandleExitDistanceDiscovery PlayerPawn))

(fn HandleWithinDistanceDiscovery
  (params (PlayerPawn :type Fort%20Player%20Pawn%20Athena%20Object%20Reference))
  (IntervalCheck|DiscoveryDistance|HandleEnterDistanceDiscovery PlayerPawn))

(fn HandleWithinDistanceMax
  (params (PlayerPawn :type Fort%20Player%20Pawn%20Athena%20Object%20Reference))
  (IntervalCheck|MaxDistance|HandleEnterDistanceMax PlayerPawn)
  (IntervalCheck|CosmeticDistance|HandleExitDistanceCosmetic PlayerPawn))

(fn HasCompletedObjective
  (params (PlayerController :type Fort%20Player%20Controller%20Object%20Reference))
  (bind _playerstate (Variables|/Script/Engine.Controller|GetPlayerState PlayerController))
  (bind _returnvalue (Actor|GetComponentbyClass _playerstate "/Script/FortniteGame.FortPlayerStateComponent_Quests"))
  (Utilities|IsValid PlayerController
    (:Is%20Valid
      (if (Class|FortQuestItemDefinition|HasComponentforWorldProcessingUpdates (Class|ConsolidatedQuestComponent|GetQuestItemDefinition (Variables|BuildingGameplayActorPropQuest|GetConsolidatedQuestComponent)))
        (Utilities|IsValid _playerstate
          (:Is%20Valid
            (Utilities|IsValid _returnvalue
              (:Is%20Valid
                (bind _returnvalue_2 (Quest|FindQuestInstanceforParticipant _returnvalue (Item|GetPersistentName (Class|ConsolidatedQuestComponent|GetQuestItemDefinition (Variables|BuildingGameplayActorPropQuest|GetConsolidatedQuestComponent)))))
                (if (Utilities|IsValidInterface _returnvalue_2)
                  (return (Class|QuestInstanceInterface|HasCompletedObjectiveInstance _returnvalue_2 (Variables|/Script/FortniteGame.ConsolidatedQuestComponent|GetObjectiveBackendName (Variables|BuildingGameplayActorPropQuest|GetConsolidatedQuestComponent))))
                  (else
                    (return false))))
              (:Is%20Not%20Valid
                _)))
          (:Is%20Not%20Valid
            _))
        (return false)
        (else
          (return (Quest|HasCompletedObjectivewithName (Quests|GetQuestManager PlayerController "Invalid") (Class|ConsolidatedQuestComponent|GetQuestItemDefinition (Variables|BuildingGameplayActorPropQuest|GetConsolidatedQuestComponent)) (Variables|/Script/FortniteGame.ConsolidatedQuestComponent|GetObjectiveBackendName (Variables|BuildingGameplayActorPropQuest|GetConsolidatedQuestComponent)) :!class /Script/FortniteGame.FortQuestManager)))))
    (:Is%20Not%20Valid
      (return false))))

(fn HiddenVisuals
  (params (PlayerController :type Fort%20Player%20Controller%20Object%20Reference))
  (Utilities|IsValid PlayerController
    (:Is%20Valid
      (if (Pawn|IsLocalPlayerController PlayerController)
        (VisualState|Actor|SetActorHidden PlayerController true)
        (Utilities|IsValid (Variables|Design|PostInteraction|GetSM_PostInteraction)
          (:Is%20Valid
            (VisualState|Meshes|SetMeshHidden (Variables|Components|GetSM_Quest_PostInteraction) PlayerController true))
          (:Is%20Not%20Valid
            _))
        (VisualState|Meshes|SetMeshHidden (Variables|Components|GetSM_Quest_PreInteraction) PlayerController true)))))

(fn IsObjectiveActive
  (params (PlayerController :type Fort%20Player%20Controller%20Object%20Reference))
  (bind _playerstate (Variables|/Script/Engine.Controller|GetPlayerState PlayerController))
  (bind _returnvalue (Actor|GetComponentbyClass _playerstate "/Script/FortniteGame.FortPlayerStateComponent_Quests"))
  (if (Class|FortQuestItemDefinition|HasComponentforWorldProcessingUpdates (Class|ConsolidatedQuestComponent|GetQuestItemDefinition (Variables|BuildingGameplayActorPropQuest|GetConsolidatedQuestComponent)))
    (Utilities|IsValid _returnvalue
      (:Is%20Valid
        (bind _returnvalue_2 (Quest|FindQuestInstanceforParticipant _returnvalue (Item|GetPersistentName (Class|ConsolidatedQuestComponent|GetQuestItemDefinition (Variables|BuildingGameplayActorPropQuest|GetConsolidatedQuestComponent)))))
        (bind _asfort_quest_item (Utilities|Casting|CastToFortQuestItem _returnvalue_2)
          (:then
            (return (Quest|IsObjectivewithNameinProgress _asfort_quest_item (Variables|/Script/FortniteGame.ConsolidatedQuestComponent|GetObjectiveBackendName (Variables|BuildingGameplayActorPropQuest|GetConsolidatedQuestComponent)) :!class /Script/FortniteGame.FortQuestItem)))
          (:CastFailed
            _)))
      (:Is%20Not%20Valid
        _))
    (return false)
    (else
      (return (Quests|IsQuestObjectiveActive (Quests|GetQuestManager PlayerController "Invalid") (Variables|/Script/FortniteGame.ConsolidatedQuestComponent|GetObjectiveBackendName (Variables|BuildingGameplayActorPropQuest|GetConsolidatedQuestComponent)))))))

(fn ManagePlayerGEs
  (params (FPC :type Fort%20Player%20Controller%20Object%20Reference))
  (bind _returnvalue (Game|Player|GetPlayerPawn FPC))
  (if (Variables|Design|QuestSetup|GameplayTags,EffectsandCosmetics|GetbShouldApplyGEs)
    (for _array_element (Variables|Design|QuestSetup|GameplayTags,EffectsandCosmetics|GetInteraction%20GEs)
      (Environment|ApplyGlobalEnvironmentGameplayEffecttoActor _returnvalue _array_element))
    (else
      _))
  (if (Variables|Design|QuestSetup|GameplayTags,EffectsandCosmetics|GetbShouldRemoveGEs)
    (for _array_element_2 (Variables|Design|QuestSetup|GameplayTags,EffectsandCosmetics|GetRemove%20GEs)
      (GameplayEffects|RemoveActiveGameplayEffectbySourceEffect (Ability|GetAbilitySystemComponent _returnvalue) _array_element_2 0 -1))
    (else
      _))
  (return))

(fn NewFunction
  (params (Interface :type Interface%20Interface%20%28by%20ref%29)))

(fn NoVFX&Audio
  (params (PlayerController :type Fort%20Player%20Controller%20Object%20Reference))
  (Utilities|IsValid PlayerController
    (:Is%20Valid
      (if (Pawn|IsLocalPlayerController PlayerController)
        (Utilities|FlowControl|Sequence
          (:then_0
            (Components|Activation|Deactivate (Variables|Components|GetP_Cosmetic)))
          (:then_1
            ;[1 1 1] Fade Out Audio
            (bind _postinteractionambientaudio (Variables|Components|GetPostInteractionAmbientAudio)
              (:then
                (Audio|Components|Audio|FadeOut _postinteractionambientaudio (Variables|Design|PostInteraction|AmbientFX|GetPostInteractionAmbientAudioFadeOut) 0.0 "Linear" :!class /Script/Engine.AudioComponent))
              (:else
                _))
            (bind _cosmeticaudio (Variables|Components|GetCosmeticAudio)
              (:then
                (Audio|Components|Audio|FadeOut _cosmeticaudio (Variables|Design|CosmeticEvent|GetCosmeticAudioFadeOut) 0.0 "Linear" :!class /Script/Engine.AudioComponent))
              (:else
                _))
            (bind _ambientaudio (Variables|Components|GetAmbientAudio)
              (:then
                (Audio|Components|Audio|FadeOut _ambientaudio (Variables|Design|PreInteraction|AmbientFX|GetAmbientAudioFadeOut) 0.0 "Linear" :!class /Script/Engine.AudioComponent))
              (:else))))))))

(fn PlayRewardFeedback
  (params (PlayerController :type Fort%20Player%20Controller%20Object%20Reference))
  (Class|AthenaQuestFunctionLibrary|AthenaPlayQuestRewardFeedback (Variables|Design|PostInteraction|SuccessFX|GetbPlayFXOnPlayer) (Variables|Design|PostInteraction|SuccessFX|GetSuccessSound) (Variables|Design|PostInteraction|SuccessFX|GetSuccessNiagara) (Variables|Design|PostInteraction|SuccessFX|GetFF_QuestRevealRumble) self (Game|Player|GetPlayerPawn PlayerController) (Variables|Design|PostInteraction|SuccessFX|GetSuccessFXTransform)))

(fn PostInteractionVFX&Audio
  (params (PlayerController :type Fort%20Player%20Controller%20Object%20Reference))
  (Utilities|IsValid PlayerController
    (:Is%20Valid
      (if (Pawn|IsLocalPlayerController PlayerController)
        (Utilities|FlowControl|Sequence
          (:then_0
            (Components|Activation|Deactivate (Variables|Components|GetP_Cosmetic))
            (Components|Activation|Activate (Variables|Components|GetP_PostInteractionAmbient) true))
          (:then_1
            ;[1 1 1] Fade Out Audio
            (bind _ambientaudio (Variables|Components|GetAmbientAudio)
              (:then
                (Audio|Components|Audio|FadeOut _ambientaudio (Variables|Design|PreInteraction|AmbientFX|GetAmbientAudioFadeOut) 0.0 "Linear" :!class /Script/Engine.AudioComponent))
              (:else
                _))
            (bind _cosmeticaudio (Variables|Components|GetCosmeticAudio)
              (:then
                (Audio|Components|Audio|FadeOut _cosmeticaudio (Variables|Design|CosmeticEvent|GetCosmeticAudioFadeOut) 0.0 "Linear" :!class /Script/Engine.AudioComponent))
              (:else
                _))
            ;[1 1 1] Fade In Post Audio
            (bind _postinteractionambientaudio (Variables|Components|GetPostInteractionAmbientAudio)
              (:then
                (Audio|Components|Audio|FadeIn _postinteractionambientaudio (Variables|Design|PostInteraction|AmbientFX|GetPostInteractionAmbientAudioFadeIn) 1.0 0.0 "Linear" :!class /Script/Engine.AudioComponent))
              (:else))))))))

(fn PostInteractionVisuals
  (params (PlayerController :type Fort%20Player%20Controller%20Object%20Reference))
  (bind _sm_quest_preinteraction (Variables|Components|GetSM_Quest_PreInteraction))
  (bind _returnvalue (Rendering|Material|GetNumMaterials _sm_quest_preinteraction))
  (bind _returnvalue_2 (- _returnvalue 1))
  (bind _advanced_postinteraction_mis_override (Variables|Design|PostInteraction|GetAdvanced_PostInteraction_MIS_Override))
  (bind _postinteraction_mis_override (Variables|Design|PostInteraction|GetPostInteraction_MIS_Override))
  (bind _sm_quest_preinteraction_2 (Variables|Components|GetSM_Quest_PreInteraction))
  (bind _storedoriginalpreinteractionmaterial (Variables|Advanced|Cosmetic|GetStoredOriginalPreInteractionMaterial))
  (bind _storedoriginalpreinteractionmaterial_2 (Variables|Advanced|Cosmetic|GetStoredOriginalPreInteractionMaterial))
  (bind _sm_quest_preinteraction_3 (Variables|Components|GetSM_Quest_PreInteraction))
  (Utilities|IsValid PlayerController
    (:Is%20Valid
      (if (Pawn|IsLocalPlayerController PlayerController)
        (VisualState|Actor|SetActorHidden PlayerController)
        (Utilities|IsValid (Variables|Design|PostInteraction|GetSM_PostInteraction)
          (:Is%20Valid
            (VisualState|Meshes|SetMeshHidden (Variables|Components|GetSM_Quest_PostInteraction) PlayerController)
            (Transformation|SetRelativeTransform (Variables|Components|GetSM_Quest_PostInteraction) (Variables|Design|PostInteraction|GetPostInteraction%20Mesh%20Transform) false false))
          (:Is%20Not%20Valid
            _))
        (VisualState|Meshes|SetMeshHidden (Variables|Components|GetSM_Quest_PreInteraction) PlayerController (Utilities|IsValid (Variables|Design|PostInteraction|GetSM_PostInteraction) :!class /Script/Engine.KismetSystemLibrary))
        (if (Variables|Design|PostInteraction|GetUseAdvancedPostMISOverride)
          (bind _keys (Utilities|Map|Keys _advanced_postinteraction_mis_override))
          (for _array_element _keys
            (if (.ReturnValue (Utilities|Map|Find _advanced_postinteraction_mis_override _array_element))
              (Utilities|IsValid (Utilities|Map|Find _advanced_postinteraction_mis_override _array_element)
                (:Is%20Valid
                  (Utilities|IsValid (Variables|Design|PostInteraction|GetSM_PostInteraction)
                    (:Is%20Valid
                      (if (not (> _array_element (- (Rendering|Material|GetNumMaterials (Variables|Components|GetSM_Quest_PostInteraction)) 1)))
                        (Rendering|Material|SetMaterial (Variables|Components|GetSM_Quest_PostInteraction) _array_element (Utilities|Map|Find _advanced_postinteraction_mis_override _array_element))))
                    (:Is%20Not%20Valid
                      (if (not (> _array_element _returnvalue_2))
                        (Rendering|Material|SetMaterial _sm_quest_preinteraction_3 _array_element (Utilities|Map|Find _advanced_postinteraction_mis_override _array_element))))))
                (:Is%20Not%20Valid
                  (Utilities|IsValid _storedoriginalpreinteractionmaterial_2
                    (:Is%20Valid
                      (if (not (> _array_element _returnvalue_2))
                        (Rendering|Material|SetMaterial _sm_quest_preinteraction_3 _array_element _storedoriginalpreinteractionmaterial_2))))))))
          (else
            (Utilities|IsValid (Variables|Design|PostInteraction|GetPostInteraction_MIS_Override)
              (:Is%20Valid
                (Utilities|IsValid (Variables|Design|PostInteraction|GetSM_PostInteraction)
                  (:Is%20Valid
                    (Rendering|Material|SetMaterial (Variables|Components|GetSM_Quest_PostInteraction) 0 _postinteraction_mis_override))
                  (:Is%20Not%20Valid
                    (Rendering|Material|SetMaterial _sm_quest_preinteraction_2 0 _postinteraction_mis_override))))
              (:Is%20Not%20Valid
                (Utilities|IsValid _storedoriginalpreinteractionmaterial
                  (:Is%20Valid
                    (Rendering|Material|SetMaterial _sm_quest_preinteraction_2 0 _storedoriginalpreinteractionmaterial)))))))))))

(fn PreInteractionVFX&Audio
  (params (PlayerController :type Fort%20Player%20Controller%20Object%20Reference))
  (Utilities|IsValid PlayerController
    (:Is%20Valid
      (if (Pawn|IsLocalPlayerController PlayerController)
        (Utilities|FlowControl|Sequence
          (:then_0
            (Components|Activation|Deactivate (Variables|Components|GetP_PostInteractionAmbient))
            (Components|Activation|Activate (Variables|Components|GetP_Ambient) false))
          (:then_1
            ;[1 1 1] Fade Out Audio
            (bind _postinteractionambientaudio (Variables|Components|GetPostInteractionAmbientAudio)
              (:then
                (Audio|Components|Audio|FadeOut _postinteractionambientaudio (Variables|Design|PostInteraction|AmbientFX|GetPostInteractionAmbientAudioFadeOut) 0.0 "Linear" :!class /Script/Engine.AudioComponent))
              (:else
                _))
            (bind _cosmeticaudio (Variables|Components|GetCosmeticAudio)
              (:then
                (Audio|Components|Audio|FadeOut _cosmeticaudio (Variables|Design|CosmeticEvent|GetCosmeticAudioFadeOut) 0.0 "Linear" :!class /Script/Engine.AudioComponent))
              (:else
                _))
            ;[1 1 1] Fade In Audio
            (bind _ambientaudio (Variables|Components|GetAmbientAudio)
              (:then
                (Audio|Components|Audio|FadeIn _ambientaudio (Variables|Design|PreInteraction|AmbientFX|GetAmbientAudioFadeIn) 1.0 0.0 "Linear" :!class /Script/Engine.AudioComponent))
              (:else))))))))

(fn PreInteractionVisuals
  (params (PlayerController :type Fort%20Player%20Controller%20Object%20Reference))
  (bind _preinteraction_mis_override (Variables|Design|PreInteraction|GetPreInteraction_MIS_Override))
  (bind _advanced_preinteraction_mis_override (Variables|Design|PreInteraction|GetAdvanced_PreInteraction_MIS_Override))
  (Utilities|IsValid PlayerController
    (:Is%20Valid
      (if (Pawn|IsLocalPlayerController PlayerController)
        (VisualState|Actor|SetActorHidden PlayerController)
        (Utilities|IsValid (Variables|Design|PostInteraction|GetSM_PostInteraction)
          (:Is%20Valid
            (VisualState|Meshes|SetMeshHidden (Variables|Components|GetSM_Quest_PostInteraction) PlayerController true))
          (:Is%20Not%20Valid
            _))
        (VisualState|Meshes|SetMeshHidden (Variables|Components|GetSM_Quest_PreInteraction) PlayerController)
        (Transformation|SetRelativeTransform (Variables|Components|GetSM_Quest_PreInteraction) (Variables|Design|PreInteraction|GetPreInteraction%20Mesh%20Transform) false false)
        (if (Variables|Design|PreInteraction|GetUseAdvancedPreMISOverride)
          (bind _keys (Utilities|Map|Keys _advanced_preinteraction_mis_override))
          (for _array_element _keys
            (if (.ReturnValue (Utilities|Map|Find _advanced_preinteraction_mis_override _array_element))
              (if (not (> _array_element (- (Rendering|Material|GetNumMaterials (Variables|Components|GetSM_Quest_PreInteraction)) 1)))
                (Rendering|Material|SetMaterial (Variables|Components|GetSM_Quest_PreInteraction) _array_element (Utilities|Map|Find _advanced_preinteraction_mis_override _array_element)))))
          (else
            (Utilities|IsValid _preinteraction_mis_override
              (:Is%20Valid
                (Rendering|Material|SetMaterial (Variables|Components|GetSM_Quest_PreInteraction) 0 _preinteraction_mis_override)))))))))

(fn RemoveWatchedActor
  (params (PlayerController :type Fort%20Player%20Controller%20Object%20Reference))
  (if (Variables|Advanced|IntervalCheck|GetbUsingDistanceCheck)
    (bind _asfort_player_pawn_athena (Utilities|Casting|CastToFortPlayerPawnAthena (Game|Player|GetPlayerPawn PlayerController)))
    (Utilities|Set|Remove (Variables|Advanced|IntervalCheck|GetWatchedActors_Max) _asfort_player_pawn_athena)
    (Utilities|Set|Remove (Variables|Advanced|IntervalCheck|GetWatchedActors_Cosmetic) _asfort_player_pawn_athena)
    (Utilities|Set|Remove (Variables|Advanced|IntervalCheck|GetWatchedActors_Discovery) _asfort_player_pawn_athena)))

(fn SetActorHidden
  (params (PlayerController :type Fort%20Player%20Controller%20Object%20Reference) (Hidden :type Boolean))
  (if (or (PlayerController|IsinSplitscreen PlayerController) false)
    ;[1 1 1] Expand the components directly to avoid traversing thousands of actor components every frame.
    (for _array_element (Actor|GetComponentsByClass "/Script/Engine.PrimitiveComponent")
      (PlayerController|SetCompHiddenControllerOnly PlayerController _array_element Hidden true))
    (else
      (Rendering|SetActorHiddenInGame Hidden))))

(fn SetMeshHidden
  (params (StaticMesh :type Static%20Mesh%20Component%20Object%20Reference) (PlayerController :type Fort%20Player%20Controller%20Object%20Reference) (Hidden :type Boolean))
  (if (PlayerController|IsinSplitscreen PlayerController)
    (PlayerController|SetCompHiddenControllerOnly PlayerController StaticMesh Hidden true)
    (else
      (Development|SetHiddeninGame StaticMesh Hidden true)))
  (Rendering|SetCastShadow StaticMesh (not Hidden)))

(fn SetQuestPropData
  (params (QuestDefinition :type Fort%20Quest%20Item%20Definition%20Object%20Reference) (ObjBackendName :type Name))
  (bind _consolidatedquestcomponent (Variables|BuildingGameplayActorPropQuest|GetConsolidatedQuestComponent))
  (Class|ConsolidatedQuestComponent|SetQuestItemDefinition _consolidatedquestcomponent QuestDefinition)
  (Class|ConsolidatedQuestComponent|SetObjectiveBackendName _consolidatedquestcomponent ObjBackendName))

(fn ShouldDie
  (params (Damage :type Float%20%28single-precision%29) (EventInstigator :type Controller%20Object%20Reference) (DamageCauser :type Actor%20Object%20Reference) (bIsTest :type Boolean) (EffectContext :type Gameplay%20Effect%20Context%20Handle%20Structure%20%28by%20ref%29))
  (return false))

(fn SpawnCosmeticActor
  (Networking|SwitchHasAuthority
    (:Remote
      (if (Utilities|IsValidClass (Variables|Design|CosmeticEvent|GetClientCosmeticActorClass))
        (Utilities|IsValid (Variables|Advanced|Cosmetic|GetClientCosmeticActorRef)
          (:Is%20Valid
            (if (Variables|Design|QuestSetup|GetbTestVisualsNoQuestRequired)
              _))
          (:Is%20Not%20Valid
            _)) ; this prevents more than 1 per player per actor
        (bind _returnvalue (Game|SpawnActorfromClass (Variables|Design|CosmeticEvent|GetClientCosmeticActorClass) (Math|Transform|MakeTransform (Transformation|GetActorLocation self) (Transformation|GetActorRotation self)) "AlwaysSpawn" "SelectDefaultAtRuntime" 0 0))
        (Variables|Advanced|Cosmetic|SetClientCosmeticActorRef _returnvalue)
        (Development|PrintString "Spawned!\\")))))

(fn UnhideVFX&Audio
  (params (PlayerController :type Fort%20Player%20Controller%20Object%20Reference))
  (Utilities|IsValid PlayerController
    (:Is%20Valid
      (if (Pawn|IsLocalPlayerController PlayerController)
        (Components|Activation|Activate (Variables|Components|GetUnHideAudio) false)
        (Components|Activation|Activate (Variables|Components|GetP_UnHide) true)))))

(fn UpdateOverlappingActors
  (bind _sm_hiddeninteractioncollision (Variables|Components|GetSM_HiddenInteractionCollision))
  (bind _overlappingactors (Collision|GetOverlappingActors _sm_hiddeninteractioncollision "/Script/FortniteGame.FortPlayerPawn" :!class /Script/Engine.PrimitiveComponent))
  (bind _collision_discoverradius (Variables|Components|GetCollision_DiscoverRadius))
  (bind _overlappingactors_2 (Collision|GetOverlappingActors _collision_discoverradius "/Script/FortniteGame.FortPlayerPawn" :!class /Script/Engine.PrimitiveComponent))
  (if (Variables|Design|QuestSetup|GetbRecheckingTriggersOnSuccess)
    _
    (elif (Variables|Design|QuestSetup|GetbReHideIfConditionsChange)
      _
      (else
        (return))))
  ;[1 0 0] Might need to do something special here for distance check if unresponsive
  (if (and (Variables|Design|GetbUseDiscoverCollisionToUnhide) (not (Variables|Advanced|IntervalCheck|GetbUsingDistanceCheck)))
    (if (Utilities|IsArrayNonEmpty _overlappingactors_2)
      (for _array_element _overlappingactors_2
        (CallFunction|UpdateOverlappingActors(CollisionDiscoverRadius) _array_element))
      (else
        _))
    (if (Utilities|IsArrayNonEmpty _overlappingactors)
      (for _array_element_2 _overlappingactors
        (CallFunction|UpdateOverlappingActors(SMHiddenInteractionCollision) _array_element_2))
      (else
        _))
    (return)
    (else
      _)))

(fn UpdateWatchedActor
  (params (PlayerPawn :type Fort%20Player%20Pawn%20Athena%20Object%20Reference))
  (bind _currentpawn (LocalVariables|GetCurrentPawn))
  (bind _output_get (LocalVariables|SetCurrentPawn PlayerPawn))
  (LocalVariables|SetActorDistanceSquared (Default|GetDistanceSquaredBetweenActors _output_get self))
  (if (> (LocalVariables|GetActorDistanceSquared) (Variables|Advanced|IntervalCheck|GetWatchSquaredDistance_Max))
    (IntervalCheck|MaxDistance|HandleExitDistanceMax (LocalVariables|GetCurrentPawn))
    (elif (> (LocalVariables|GetActorDistanceSquared) (Variables|Advanced|IntervalCheck|GetWatchSquaredDistance_Cosmetic))
      (IntervalCheck|MaxDistance|HandleWithinDistanceMax (LocalVariables|GetCurrentPawn))
      (elif (or (> (LocalVariables|GetActorDistanceSquared) (Variables|Advanced|IntervalCheck|GetWatchSquaredDistance_Discovery)) (not (Variables|Design|GetbUseDiscoverCollisionToUnhide)))
        (IntervalCheck|CosmeticDistance|HandleWithinDistanceCosmetic (LocalVariables|GetCurrentPawn))
        (else
          (IntervalCheck|DiscoveryDistance|HandleWithinDistanceDiscovery (LocalVariables|GetCurrentPawn))))))
  (if (Vehicle|IsinVehicle _currentpawn)
    (return (Math|Vector|VectorLength (Vehicle|GetVehicleVelocity (Vehicle|GetVehicleInterface _currentpawn))))
    (else
      (return 5.5e+02))))

(fn UpdateWatchedActorsDistanceCosmetic
  (bind _highestspeed (LocalVariables|GetHighestSpeed))
  (bind _result (Utilities|Set|ToArray (Variables|Advanced|IntervalCheck|GetWatchedActors_Cosmetic)))
  (for _array_element _result
    (bind _pawnspeed (IntervalCheck|UpdateWatchedActor _array_element))
    (LocalVariables|SetHighestSpeed (Math|Float|Max(Float) _pawnspeed (LocalVariables|GetHighestSpeed)))
    (if (Utilities|Set|ContainsItem (Variables|Advanced|IntervalCheck|GetWatchedActors_Cosmetic) _array_element)
      (CallFunction|OnIntervalWithinDistanceCosmetic (Fortnite|GetFortPlayerControllerfromActor _array_element) _array_element)))
  (CallFunction|CheckNextIntervalDistanceCosmetic _highestspeed)
  (if (Variables|Design|GetbUseDiscoverCollisionToUnhide)
    (CallFunction|CheckNextIntervalDistanceDiscovery _highestspeed)))

(fn UpdateWatchedActorsDistanceDiscovery
  (if (Variables|Design|GetbUseDiscoverCollisionToUnhide)
    (bind _result (Utilities|Set|ToArray (Variables|Advanced|IntervalCheck|GetWatchedActors_Discovery)))
    (for _array_element _result
      (bind _pawnspeed (IntervalCheck|UpdateWatchedActor _array_element))
      (LocalVariables|SetHighestSpeed (Math|Float|Max(Float) _pawnspeed (LocalVariables|GetHighestSpeed)))
      (if (Utilities|Set|ContainsItem (Variables|Advanced|IntervalCheck|GetWatchedActors_Discovery) _array_element)
        (CallFunction|OnIntervalWithinDistanceDiscovery (Fortnite|GetFortPlayerControllerfromActor _array_element) _array_element)))
    (CallFunction|CheckNextIntervalDistanceDiscovery (LocalVariables|GetHighestSpeed))))

(fn UpdateWatchedActorsDistanceMax
  (bind _highestspeed (LocalVariables|GetHighestSpeed))
  (bind _result (Utilities|Set|ToArray (Variables|Advanced|IntervalCheck|GetWatchedActors_Max)))
  (for _array_element _result
    (bind _pawnspeed (IntervalCheck|UpdateWatchedActor _array_element))
    (LocalVariables|SetHighestSpeed (Math|Float|Max(Float) _pawnspeed (LocalVariables|GetHighestSpeed)))
    (if (Utilities|Set|ContainsItem (Variables|Advanced|IntervalCheck|GetWatchedActors_Max) _array_element)
      (CallFunction|OnIntervalWithinDistanceMax (Fortnite|GetFortPlayerControllerfromActor _array_element) _array_element)))
  (CallFunction|CheckNextIntervalDistanceMax _highestspeed)
  (CallFunction|CheckNextIntervalDistanceCosmetic _highestspeed)
  (if (Variables|Design|GetbUseDiscoverCollisionToUnhide)
    (CallFunction|CheckNextIntervalDistanceDiscovery _highestspeed)))

(fn ConstructionScript
  (bind (_origin _boxextent) (Collision|GetComponentBounds (Variables|Components|GetSM_HiddenInteractionCollision)))
  (bind (_origin_2 _boxextent_2) (Collision|GetComponentBounds (Variables|Components|GetSM_Quest_PreInteraction)))
  (bind _ambientsound (Variables|Design|PreInteraction|AmbientFX|GetAmbientSound))
  (bind _sm_hiddeninteractioncollision (Variables|Components|GetSM_HiddenInteractionCollision))
  (bind (_relativelocation_x _relativelocation_y) (Math|Vector|BreakVector (Variables|/Script/Engine.SceneComponent|GetRelativeLocation_X (Variables|Components|GetText_BackendNameHelper))))
  (bind _sm_hiddeninteractioncollision_2 (Variables|Components|GetSM_HiddenInteractionCollision))
  (bind _questiconcomponent (Variables|BuildingGameplayActorPropQuest|GetQuestIconComponent))
  (bind _questiconcomponent_2 (Variables|BuildingGameplayActorPropQuest|GetQuestIconComponent))
  (bind _p_postinteractionambient (Variables|Components|GetP_PostInteractionAmbient))
  (bind _n_ambient (Variables|Components|GetN_Ambient))
  (bind _sm_hiddeninteractioncollision_3 (Variables|Components|GetSM_HiddenInteractionCollision))
  (bind _sm_hiddeninteractioncollision_4 (Variables|Components|GetSM_HiddenInteractionCollision))
  (bind _p_ambient (Variables|Components|GetP_Ambient))
  (bind _preinteraction_mis_override (Variables|Design|PreInteraction|GetPreInteraction_MIS_Override))
  (bind _postinteractionambientsound (Variables|Design|PostInteraction|AmbientFX|GetPostInteractionAmbientSound))
  (bind _cosmeticsound (Variables|Design|CosmeticEvent|GetCosmeticSound))
  (bind _sm_hiddeninteractioncollision_5 (Variables|Components|GetSM_HiddenInteractionCollision))
  (bind _unhide_particle (Variables|Design|PreInteraction|UnhideFX|GetUnhide%20Particle))
  (bind _unhidesound (Variables|Design|PreInteraction|UnhideFX|GetUnhideSound))
  (bind _postinteraction_mis_override (Variables|Design|PostInteraction|GetPostInteraction_MIS_Override))
  (bind _sm_hiddeninteractioncollision_6 (Variables|Components|GetSM_HiddenInteractionCollision))
  (bind _ambientniagarasystem (Variables|Design|PreInteraction|AmbientFX|GetAmbientNiagaraSystem))
  (bind _sm_hiddeninteractioncollision_7 (Variables|Components|GetSM_HiddenInteractionCollision))
  (bind _unhideniagarasystem (Variables|Design|PreInteraction|UnhideFX|GetUnhideNiagaraSystem))
  (bind _postinteractionambientniagarasystem (Variables|Design|PostInteraction|AmbientFX|GetPostInteractionAmbientNiagaraSystem))
  (bind _ambientparticle (Variables|Design|PreInteraction|AmbientFX|GetAmbientParticle))
  (bind _postinteractionambientparticle (Variables|Design|PostInteraction|AmbientFX|GetPostInteractionAmbientParticle))
  (bind _cosmeticparticle (Variables|Design|CosmeticEvent|GetCosmeticParticle))
  (bind _cosmeticniagarasystem (Variables|Design|CosmeticEvent|GetCosmeticNiagaraSystem))
  (bind _sm_hiddeninteractioncollision_8 (Variables|Components|GetSM_HiddenInteractionCollision))
  (Utilities|FlowControl|Sequence
    (:then_0
      ;[1 1 1] Master Quest Type Init
      (Variables|Design|QuestSetup|GameplayTags,EffectsandCosmetics|SetbShouldApplyGEs (Utilities|IsArrayNonEmpty (Variables|Design|QuestSetup|GameplayTags,EffectsandCosmetics|GetInteraction%20GEs)))
      (Variables|Design|QuestSetup|GameplayTags,EffectsandCosmetics|SetbShouldRemoveGEs (Utilities|IsArrayNonEmpty (Variables|Design|QuestSetup|GameplayTags,EffectsandCosmetics|GetRemove%20GEs)))
      (Variables|Design|QuestSetup|GameplayTags,EffectsandCosmetics|SetbShouldApplyCosmeticOverrides (Utilities|IsArrayNonEmpty (Variables|Design|QuestSetup|GameplayTags,EffectsandCosmetics|GetPlayer%20Cosmetic%20Overrides)))
      (switch Utilities|FlowControl|Switch|SwitchonE_Athena_GenericQuestActorTypes (Variables|Design|QuestSetup|GetbQuestActorType)
        (:NewEnumerator0
          (Variables|Design|SetbCompleteFromDiscovery false)
          (Variables|Design|SetbUseDiscoverCollisionToUnhide false))
        (:NewEnumerator1
          (Variables|Design|SetbCompleteFromDiscovery false)
          (Variables|Design|SetbUseDiscoverCollisionToUnhide true))
        (:NewEnumerator2
          (Variables|Design|SetbCompleteFromDiscovery true)
          (Variables|Design|SetbUseDiscoverCollisionToUnhide true))))
    (:then_1
      ;[1 1 1] Pre and Post Interaction Meshes Init
      (Components|StaticMesh|SetStaticMesh (Variables|Components|GetSM_Quest_PreInteraction) (Variables|Design|PreInteraction|GetSM_PreInteraction))
      (Components|StaticMesh|SetStaticMesh (Variables|Components|GetSM_Quest_PostInteraction) (Variables|Design|PostInteraction|GetSM_PostInteraction))
      (Transformation|SetRelativeTransform (Variables|Components|GetSM_Quest_PostInteraction) (Variables|Design|PostInteraction|GetPostInteraction%20Mesh%20Transform) false false)
      (Transformation|SetRelativeTransform (Variables|Components|GetSM_Quest_PreInteraction) (Variables|Design|PreInteraction|GetPreInteraction%20Mesh%20Transform) false false)
      ;;[1 1 1] Hide Or Not Post Interaction Mesh
      (if (Variables|Design|PostInteraction|GetbPreviewPostInteraction)
        (Utilities|IsValid (Variables|Design|PostInteraction|GetSM_PostInteraction)
          (:Is%20Valid
            (Transformation|SetWorldScale3D (Variables|Components|GetSM_Quest_PreInteraction) "0, 0, 0"))
          (:Is%20Not%20Valid
            _))
        (else
          (Transformation|SetWorldScale3D (Variables|Components|GetSM_Quest_PostInteraction) "0, 0, 0")))
      ;;[1 1 1] Change based on if we are using "discovery" or not
      (if (Variables|Design|GetbUseDiscoverCollisionToUnhide)
        (Transformation|SetWorldScale3D (Variables|Components|GetSM_Quest_PreInteraction) "0, 0, 0")
        (else
          (Components|DestroyComponent (Variables|Components|GetCollision_DiscoverRadius)))))
    (:then_2
      ;[1 1 1] Particle Init
      (Transformation|SetRelativeTransform (Variables|Components|GetP_Ambient) (Variables|Design|PreInteraction|AmbientFX|GetAmbient%20Particle%20Transform) false true)
      (Utilities|IsValid _ambientparticle
        (:Is%20Valid
          (Effects|Components|ParticleSystem|SetTemplate (Variables|Components|GetP_Ambient) _ambientparticle))
        (:Is%20Not%20Valid
          _))
      (Utilities|IsValid _ambientniagarasystem
        (:Is%20Valid
          (Niagara|SetNiagaraSystemAsset (Variables|Components|GetN_Ambient) _ambientniagarasystem true)
          (Effects|Components|ParticleSystem|SetBoolParameter _n_ambient "FXColorRed" (Variables|Design|PreInteraction|AmbientFX|GetAmbientNiagaraIsRed))
          (Effects|Components|ParticleSystem|SetFloatParameter _n_ambient "FXDistance" (Variables|Design|QuestSetup|GetVFXFullRadius))
          (Effects|Components|ParticleSystem|SetFloatParameter _n_ambient "FXScale" (Variables|Design|PreInteraction|AmbientFX|GetAmbientNiagaraLinearScaling))
          (Effects|Components|ParticleSystem|SetVectorParameter _n_ambient "FXScaleXYZ" (Variables|Design|PreInteraction|AmbientFX|GetAmbientNiagaraVectorScaling)))
        (:Is%20Not%20Valid
          _))
      (Utilities|IsValid _postinteractionambientniagarasystem
        (:Is%20Valid
          (Niagara|SetNiagaraSystemAsset (Variables|Components|GetN_PostInteractionAmbient) _postinteractionambientniagarasystem true))
        (:Is%20Not%20Valid
          _))
      (Utilities|IsValid _postinteractionambientparticle
        (:Is%20Valid
          (Effects|Components|ParticleSystem|SetTemplate (Variables|Components|GetP_PostInteractionAmbient) _postinteractionambientparticle))
        (:Is%20Not%20Valid
          _))
      (Transformation|SetRelativeTransform (Variables|Components|GetP_PostInteractionAmbient) (Variables|Design|PostInteraction|AmbientFX|GetPostInteractionAmbientParticleTransform) false false)
      (if (Variables|Design|QuestSetup|GetbTestVisualsNoQuestRequired)
        (if (Variables|Design|PostInteraction|GetbPreviewPostInteraction)
          (Components|Activation|SetAutoActivate _p_ambient false)
          (Components|Activation|SetAutoActivate _p_postinteractionambient true)
          (else
            (Components|Activation|SetAutoActivate _p_ambient true)
            (Components|Activation|SetAutoActivate _p_postinteractionambient false)))
        (else
          _))
      (Utilities|IsValid _cosmeticparticle
        (:Is%20Valid
          (Effects|Components|ParticleSystem|SetTemplate (Variables|Components|GetP_Cosmetic) _cosmeticparticle))
        (:Is%20Not%20Valid
          _))
      (Utilities|IsValid _cosmeticniagarasystem
        (:Is%20Valid
          (Niagara|SetNiagaraSystemAsset (Variables|Components|GetN_Cosmetic) _cosmeticniagarasystem true))
        (:Is%20Not%20Valid
          _))
      (Transformation|SetRelativeTransform (Variables|Components|GetP_Cosmetic) (Variables|Design|CosmeticEvent|GetCosmeticParticleTransform) false false)
      (Utilities|IsValid _unhide_particle
        (:Is%20Valid
          (Effects|Components|ParticleSystem|SetTemplate (Variables|Components|GetP_UnHide) _unhide_particle))
        (:Is%20Not%20Valid
          _))
      (Utilities|IsValid _unhideniagarasystem
        (:Is%20Valid
          (Niagara|SetNiagaraSystemAsset (Variables|Components|GetN_UnHide) _unhideniagarasystem true))
        (:Is%20Not%20Valid
          _))
      (Transformation|SetRelativeTransform (Variables|Components|GetN_UnHide) (Variables|Design|PreInteraction|UnhideFX|GetUnhideParticleTransform) false false))
    (:then_3
      ;[1 1 1] Sound Init
      (Utilities|IsValid _ambientsound
        (:Is%20Valid
          (Audio|Components|Audio|SetSound (Variables|Components|GetAmbientAudio) _ambientsound))
        (:Is%20Not%20Valid
          _))
      (Utilities|IsValid _postinteractionambientsound
        (:Is%20Valid
          (Audio|Components|Audio|SetSound (Variables|Components|GetPostInteractionAmbientAudio) _postinteractionambientsound))
        (:Is%20Not%20Valid
          _))
      (Utilities|IsValid _cosmeticsound
        (:Is%20Valid
          (Audio|Components|Audio|SetSound (Variables|Components|GetCosmeticAudio) _cosmeticsound))
        (:Is%20Not%20Valid
          _))
      (Utilities|IsValid _unhidesound
        (:Is%20Valid
          (Audio|Components|Audio|SetSound (Variables|Components|GetUnHideAudio) _unhidesound))))
    (:then_4
      ;[1 1 1] Collision Setup
      (if (ScalableFloat|GetValueasBool (Variables|Advanced|IntervalCheck|GetEnableDistanceCheck))
        (Components|DestroyComponent (Variables|Components|GetCollision_DiscoverRadius))
        (Variables|Advanced|IntervalCheck|SetbUsingDistanceCheck true)
        (else
          (Components|Sphere|SetSphereRadius (Variables|Components|GetCollision_DiscoverRadius) (Variables|Design|QuestSetup|GetDiscoverRadius) false))))
    (:then_5
      ;[1 1 1] MiniMap Setup
      (if (or (Variables|Design|QuestSetup|MinimapIcon|GetbShowMinimapIcon) (Variables|Design|QuestSetup|MinimapIcon|GetbShowBigMapIcon))
        (MiniMapData|SetIcon _questiconcomponent (Variables|Design|QuestSetup|MinimapIcon|GetMini%20Map%20Icon))
        (MiniMapData|SetIconPinned _questiconcomponent (Variables|Design|QuestSetup|MinimapIcon|GetMini%20Map%20Icon%20Pinned))
        (switch Utilities|FlowControl|Switch|SwitchonE_Athena_GenericQuestActorMinimapRanges (Variables|Design|QuestSetup|MinimapIcon|GetMapRevealDistance)
          (:NewEnumerator0
            (Variables|Design|QuestSetup|MinimapIcon|SetView%20Distance 1024.0))
          (:NewEnumerator1
            (Variables|Design|QuestSetup|MinimapIcon|SetView%20Distance 2.5e+03))
          (:NewEnumerator2
            (Variables|Design|QuestSetup|MinimapIcon|SetView%20Distance 5e+03))
          (:NewEnumerator3
            (Variables|Design|QuestSetup|MinimapIcon|SetView%20Distance 7.5e+03))
          (:NewEnumerator4
            (Variables|Design|QuestSetup|MinimapIcon|SetView%20Distance 1.2e+04))
          (:NewEnumerator5
            (Variables|Design|QuestSetup|MinimapIcon|SetView%20Distance (Variables|Design|QuestSetup|MinimapIcon|GetCustomViewDistance))))
        (MiniMapData|SetViewDistanceRadius (Variables|BuildingGameplayActorPropQuest|GetQuestIconComponent) (Variables|Design|QuestSetup|MinimapIcon|GetView%20Distance))
        (MiniMapData|SetIconVisibilityonMiniMap _questiconcomponent_2 (Variables|Design|QuestSetup|MinimapIcon|GetbShowMinimapIcon))
        (MiniMapData|SetIconVisibilityonMap _questiconcomponent_2 (Variables|Design|QuestSetup|MinimapIcon|GetbShowBigMapIcon))
        (else
          (Components|DestroyComponent (Variables|BuildingGameplayActorPropQuest|GetQuestIconComponent)))))
    (:then_6
      ;[1 1 1] Material Swaps
      (Utilities|IsValid _preinteraction_mis_override
        (:Is%20Valid
          (Variables|Advanced|Cosmetic|SetStoredOriginalPreInteractionMaterial (Utilities|Array|Get(acopy) (Rendering|Material|GetMaterials (Variables|Components|GetSM_Quest_PreInteraction))))
          (if (not (Variables|Design|PostInteraction|GetbPreviewPostInteraction))
            ;;[1 1 1] If PreInteraction has swapped material
            (Rendering|Material|SetMaterial (Variables|Components|GetSM_Quest_PreInteraction) 0 _preinteraction_mis_override)))
        (:Is%20Not%20Valid
          (if (Variables|Design|PostInteraction|GetbPreviewPostInteraction)
            _)))
      ;;[1 1 1] If Previewing the Post Interaction State
      (Utilities|IsValid _postinteraction_mis_override
        (:Is%20Valid
          (Utilities|IsValid (Variables|Design|PostInteraction|GetSM_PostInteraction)
            (:Is%20Valid
              (Rendering|Material|SetMaterial (Variables|Components|GetSM_Quest_PostInteraction) 0 _postinteraction_mis_override))
            (:Is%20Not%20Valid
              (Rendering|Material|SetMaterial (Variables|Components|GetSM_Quest_PreInteraction) 0 _postinteraction_mis_override))))
        (:Is%20Not%20Valid
          (Rendering|Material|SetMaterial (Variables|Components|GetSM_Quest_PreInteraction) 0 (Variables|Advanced|Cosmetic|GetStoredOriginalPreInteractionMaterial)))))
    (:then_7
      ;[1 1 1] Gross Damage to Interact Switcher
      ;;[1 1 1] Init
      (Collision|SetCollisionProfileName _sm_hiddeninteractioncollision_3 "FortTriggerOnlyInteractions" true :!class /Script/Engine.PrimitiveComponent)
      (Transformation|SetWorldScale3D _sm_hiddeninteractioncollision_3 (Variables|Design|QuestSetup|InteractionSettings|GetInteraction%20Collision%20Scale))
      (Variables|/Script/Engine.PrimitiveComponent|SetbGenerateOverlapEvents false _sm_hiddeninteractioncollision_3)
      (Variables|Advanced|Interaction|SetbOverlapToInteract false)
      (Variables|Advanced|Interaction|SetbDamageToInteract false)
      (Variables|Advanced|Interaction|SetbDepositToInteract false)
      (Variables|Advanced|Interaction|SetbListenToTagToInteract false)
      (Variables|Design|SetbConversationToComplete false)
      (switch Utilities|FlowControl|Switch|SwitchonE_Athena_GenericQuestDamageTypes (Variables|Design|QuestSetup|InteractionSettings|GetInteractionType)
        (:NewEnumerator0
          ;;[1 1 1] Tag Templates
          (Variables|Design|QuestSetup|InteractionSettings|SetDamage%20or%20Interact%20Tag "(GameplayTags=)"))
        (:NewEnumerator5
          (Variables|Design|QuestSetup|InteractionSettings|SetDamage%20or%20Interact%20Tag "(GameplayTags=((TagName=\"Weapon\")))"))
        (:NewEnumerator12
          (Variables|Design|QuestSetup|InteractionSettings|SetDamage%20or%20Interact%20Tag "(GameplayTags=((TagName=\"Weapon.Melee.Impact.Pickaxe\")))")
          ;;[1 1 1] Collision Switching
          (Collision|SetCollisionResponsetoAllChannels _sm_hiddeninteractioncollision_7 "ECR_Ignore")
          (Collision|SetCollisionResponsetoChannel _sm_hiddeninteractioncollision_7 "ECC_GameTraceChannel3" "ECR_Overlap")
          ;;[1 1 1] Logic Mode Switching
          (Variables|Advanced|Interaction|SetbDamageToInteract true))
        (:NewEnumerator6
          (Variables|Design|QuestSetup|InteractionSettings|SetDamage%20or%20Interact%20Tag "(GameplayTags=((TagName=\"Weapon.Melee\")))"))
        (:NewEnumerator7
          (Variables|Design|QuestSetup|InteractionSettings|SetDamage%20or%20Interact%20Tag "(GameplayTags=((TagName=\"Weapon.Ranged\")))"))
        (:NewEnumerator8
          (Variables|Design|QuestSetup|InteractionSettings|SetDamage%20or%20Interact%20Tag "(GameplayTags=((TagName=\"Weapon.Ranged.Grenade.Standard\"),(TagName=\"Weapon.Ranged.Heavy.C4\"),(TagName=\"Weapon.Ranged.Heavy.Grenade\"),(TagName=\"Weapon.Ranged.Heavy.Grenade_Launcher\"),(TagName=\"Weapon.Ranged.Heavy.Mortar\"),(TagName=\"Weapon.Ranged.Heavy.ProximityMine\"),(TagName=\"Weapon.Ranged.Heavy.Rocket_Launcher\")))")
          (Collision|SetCollisionProfileName _sm_hiddeninteractioncollision_4 "FortTriggerOnlyWeapons" true :!class /Script/Engine.PrimitiveComponent)
          (Variables|/Script/Engine.PrimitiveComponent|SetbGenerateOverlapEvents true _sm_hiddeninteractioncollision_4)
          (Variables|Advanced|Interaction|SetbDamageToInteract true))
        (:NewEnumerator9
          (Variables|Design|QuestSetup|InteractionSettings|SetDamage%20or%20Interact%20Tag "(GameplayTags=)")
          (Collision|SetCollisionProfileName _sm_hiddeninteractioncollision "FortTriggerOnlyPawns" true :!class /Script/Engine.PrimitiveComponent)
          (Collision|SetCollisionResponsetoAllChannels _sm_hiddeninteractioncollision "ECR_Ignore")
          (Collision|SetCollisionResponsetoChannel _sm_hiddeninteractioncollision "ECC_Vehicle" "ECR_Overlap")
          (Variables|/Script/Engine.PrimitiveComponent|SetbGenerateOverlapEvents true _sm_hiddeninteractioncollision)
          (Variables|Advanced|Interaction|SetbOverlapToInteract true))
        (:NewEnumerator10
          (Variables|Design|QuestSetup|InteractionSettings|SetDamage%20or%20Interact%20Tag "(GameplayTags=)"))
        (:NewEnumerator11
          _)
        (:NewEnumerator13
          (Variables|Advanced|Interaction|SetResourceItemDefinition "/Game/Items/ResourcePickups/Athena_WadsItemData.Athena_WadsItemData")
          (Variables|Design|QuestSetup|InteractionSettings|SetDamage%20or%20Interact%20Tag "(GameplayTags=)")
          (Variables|Advanced|Interaction|SetbDepositToInteract true))
        (:NewEnumerator14
          (Variables|Advanced|Interaction|SetResourceItemDefinition "/Game/Items/ResourcePickups/WoodItemData.WoodItemData"))
        (:NewEnumerator15
          (Variables|Advanced|Interaction|SetResourceItemDefinition "/Game/Items/ResourcePickups/StoneItemData.StoneItemData"))
        (:NewEnumerator16
          (Variables|Advanced|Interaction|SetResourceItemDefinition "/Game/Items/ResourcePickups/MetalItemData.MetalItemData"))
        (:NewEnumerator17
          (Collision|SetCollisionProfileName _sm_hiddeninteractioncollision_2 "FortTriggerOnlyPawns" true :!class /Script/Engine.PrimitiveComponent)
          (Collision|SetCollisionResponsetoAllChannels _sm_hiddeninteractioncollision_2 "ECR_Ignore")
          (Collision|SetCollisionResponsetoChannel _sm_hiddeninteractioncollision_2 "ECC_Pawn" "ECR_Overlap")
          (Variables|/Script/Engine.PrimitiveComponent|SetbGenerateOverlapEvents true _sm_hiddeninteractioncollision_2)
          (Variables|Advanced|Interaction|SetbOverlapToInteract true)
          (Variables|Advanced|Interaction|SetbListenToTagToInteract true))
        (:NewEnumerator18
          (Collision|SetCollisionProfileName _sm_hiddeninteractioncollision_8 "FortProjectileBall" true :!class /Script/Engine.PrimitiveComponent)
          (Collision|SetCollisionResponsetoAllChannels _sm_hiddeninteractioncollision_8 "ECR_Ignore")
          (Collision|SetCollisionResponsetoChannel _sm_hiddeninteractioncollision_8 "ECC_GameTraceChannel8" "ECR_Overlap")
          (Variables|/Script/Engine.PrimitiveComponent|SetbGenerateOverlapEvents true _sm_hiddeninteractioncollision_8)
          (Variables|Advanced|Interaction|SetbOverlapToInteract true))
        (:NewEnumerator19
          (Variables|Design|SetbConversationToComplete true))
        (:NewEnumerator20
          (Collision|SetCollisionProfileName _sm_hiddeninteractioncollision_6 "FortTriggerOnlyPawns" true :!class /Script/Engine.PrimitiveComponent)
          (Collision|SetCollisionResponsetoAllChannels _sm_hiddeninteractioncollision_6 "ECR_Ignore")
          (Collision|SetCollisionResponsetoChannel _sm_hiddeninteractioncollision_6 "ECC_Pawn" "ECR_Overlap")
          (Variables|/Script/Engine.PrimitiveComponent|SetbGenerateOverlapEvents true _sm_hiddeninteractioncollision_6)
          (Variables|Advanced|Interaction|SetbOverlapToInteract true))
        (:NewEnumerator21
          (Collision|SetCollisionProfileName _sm_hiddeninteractioncollision_5 "FortTriggerOnlyPawns" true :!class /Script/Engine.PrimitiveComponent)
          (Collision|SetCollisionResponsetoAllChannels _sm_hiddeninteractioncollision_5 "ECR_Ignore")
          (Collision|SetCollisionResponsetoChannel _sm_hiddeninteractioncollision_5 "ECC_Pawn" "ECR_Overlap")
          (Collision|SetCollisionResponsetoChannel _sm_hiddeninteractioncollision_5 "ECC_Vehicle" "ECR_Overlap")
          (Variables|/Script/Engine.PrimitiveComponent|SetbGenerateOverlapEvents true _sm_hiddeninteractioncollision_5)
          (Variables|Advanced|Interaction|SetbOverlapToInteract true)))
      (Collision|SetCollisionProfileName _sm_hiddeninteractioncollision_7 "FortTriggerOnlyWeapons" true :!class /Script/Engine.PrimitiveComponent)
      (Variables|Advanced|Interaction|SetbDamageToInteract true))
    (:then_8
      ;[1 1 1] Editor Only
      (Variables|Advanced|Debug|SetbErrorMessage false)
      (Variables|Advanced|Debug|SetDebugText "")
      (if (Variables|Design|QuestSetup|GetbTestVisualsNoQuestRequired)
        (Variables|Advanced|Debug|SetbErrorMessage true)
        (Variables|Advanced|Debug|SetDebugText (Utilities|String|Append (Variables|Advanced|Debug|GetDebugText) "
Visual Test Mode"))
        (else
          _))
      (if (Variables|Design|PostInteraction|GetbPreviewPostInteraction)
        (Variables|Advanced|Debug|SetbErrorMessage true)
        (Variables|Advanced|Debug|SetDebugText (Utilities|String|Append (Variables|Advanced|Debug|GetDebugText) "
Preview Post-Interaction"))
        (else
          _))
      (Utilities|IsValid (Class|ConsolidatedQuestComponent|GetQuestItemDefinition (Variables|BuildingGameplayActorPropQuest|GetConsolidatedQuestComponent))
        (:Is%20Valid
          _)
        (:Is%20Not%20Valid
          (Variables|Advanced|Debug|SetbErrorMessage true)
          (Variables|Advanced|Debug|SetDebugText (Utilities|String|Append (Variables|Advanced|Debug|GetDebugText) "
Need QuestItemDefinition"))))
      (if (not (>= (Utilities|String|Len (Utilities|String|ToString(Name) (Variables|/Script/FortniteGame.ConsolidatedQuestComponent|GetObjectiveBackendName (Variables|BuildingGameplayActorPropQuest|GetConsolidatedQuestComponent)))) 1))
        (Variables|Advanced|Debug|SetbErrorMessage true)
        (Variables|Advanced|Debug|SetDebugText (Utilities|String|Append (Variables|Advanced|Debug|GetDebugText) "
Need Objective Backend Name")))
      (if (and (Variables|Advanced|Interaction|GetbListenToTagToInteract) (not (Utilities|GameplayTagContainerNotEmpty (Variables|Design|QuestSetup|InteractionSettings|GetDamage%20or%20Interact%20Tag))))
        (Variables|Advanced|Debug|SetbErrorMessage true)
        (Variables|Advanced|Debug|SetDebugText (Utilities|String|Append (Variables|Advanced|Debug|GetDebugText) "
Need a Tag to listen to"))
        (else
          _))
      ;;[1 1 1] Positioning the Tooltip
      (Transformation|SetRelativeLocationAndRotation (Variables|Components|GetText_BackendNameHelper) (Math|Vector|MakeVector _relativelocation_x _relativelocation_y (* (Math|Float|Max(Float) (.z _boxextent_2) (.z _boxextent)) 2.0)) (Math|Rotator|InvertRotator (Transformation|GetActorRotation)) false false)
      (if (Variables|Advanced|Debug|GetbErrorMessage)
        (Rendering|Components|TextRender|SetTextRenderColor (Variables|Components|GetText_BackendNameHelper) (Variables|Advanced|Debug|GetErrorTextColor))
        (Rendering|Components|TextRender|SetText (Variables|Components|GetText_BackendNameHelper) (Utilities|Text|ToText(String) (Variables|Advanced|Debug|GetDebugText)))
        (else
          (Rendering|Components|TextRender|SetTextRenderColor (Variables|Components|GetText_BackendNameHelper) (Variables|Advanced|Debug|GetNormalTextColor))
          (Rendering|Components|TextRender|SetText (Variables|Components|GetText_BackendNameHelper) (Utilities|Text|ToText(String) (Utilities|String|Append "Obj: " (Utilities|String|ToString(Name) (Variables|/Script/FortniteGame.ConsolidatedQuestComponent|GetObjectiveBackendName (Variables|BuildingGameplayActorPropQuest|GetConsolidatedQuestComponent))) "
Tag: " (GameplayTags|GetDebugStringfromGameplayTagContainer (CallFunction|GetQuestTargetTags self)))))))
      (if (Variables|Design|QuestSetup|GetbTestVisualsNoQuestRequired)
        (Development|PrintString (Utilities|String|Append "Do Not Save with \"TEST VISUALS\" flag on!   " (Utilities|GetDisplayName self)) true true "(R=1.000000,G=0.000000,B=0.000000,A=1.000000)" 1.0)))
    (:then_10
      ;[1 1 1] Enable Conversation Component
      (if (not (Variables|Design|GetbConversationToComplete))
        (Components|DestroyComponent (Variables|Components|GetNonPlayerConversationComponent))))))
