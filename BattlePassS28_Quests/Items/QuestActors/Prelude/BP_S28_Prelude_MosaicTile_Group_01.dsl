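; Season 28 Prelude quest actor for a group mosaic tile pickup, reported when a player pawn overlaps it.
; If the pawn can interact and has authority, it broadcasts a GroupTile pickup custom verb message, adding the GroupTile quest token tag to the TilePickup subject.
; The logic appears under both BeginOverlap and a custom handler, which binds itself only when run.
; The BeginOverlap path casts a stale reference to that handler's OtherActor, not its own parameter.
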(event EventBeginPlay
  (|CallParentFunction) ; This node is disabled and will not be called.\nDrag off pins to build functionality.
)

(event EventTick
  (params (DeltaSeconds :type Float%20%28single-precision%29)))

(event Collision|EventActorBeginOverlap
  (params (OtherActor :type Actor%20Object%20Reference))
  (bind _asfort_player_pawn (Utilities|Casting|CastToFortPlayerPawn (stale Custom|OnActorBeginOverlap_Event OtherActor)))
  (if (Interaction|BlueprintCanInteract _asfort_player_pawn "FirstInteraction" "IT_NoInteraction")
    (if (Networking|HasAuthority _asfort_player_pawn)
      (bind _result (CustomVerb|AddTagsforSubject (CustomVerb|MakeCustomVerbMessage _asfort_player_pawn "(TagName=\"CustomVerb.Prelude.Pickup.GroupTile\")") "TilePickup" (GameplayTags|MakeGameplayTagContainerfromTag "(TagName=\"Athena.Quests.S28.Token.Prelude.GroupTile\")")))
      (CustomVerb|BroadcastCustomVerbMessage _result))))

(event Custom|OnActorBeginOverlap_Event
  (params (OverlappedActor :type Actor%20Object%20Reference) (OtherActor :type Actor%20Object%20Reference))
  (Collision|BindEventtoOnActorBeginOverlap (delegate Custom|OnActorBeginOverlap_Event))
  (bind _asfort_player_pawn (Utilities|Casting|CastToFortPlayerPawn OtherActor))
  (if (Interaction|BlueprintCanInteract _asfort_player_pawn "FirstInteraction" "IT_NoInteraction")
    (if (Networking|HasAuthority _asfort_player_pawn)
      (bind _result (CustomVerb|AddTagsforSubject (CustomVerb|MakeCustomVerbMessage _asfort_player_pawn "(TagName=\"CustomVerb.Prelude.Pickup.GroupTile\")") "TilePickup" (GameplayTags|MakeGameplayTagContainerfromTag "(TagName=\"Athena.Quests.S28.Token.Prelude.GroupTile\")")))
      (CustomVerb|BroadcastCustomVerbMessage _result))))

(fn ConstructionScript
  (|CallParentFunction))
