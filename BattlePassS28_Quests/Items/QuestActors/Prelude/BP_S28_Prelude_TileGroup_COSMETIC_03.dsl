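; Cosmetic tile group actor for the Season 28 Prelude quest content that plays a one-shot animation on a tile mesh, then destroys itself.
; On BeginPlay it stores the tile's relative location and rotation and calls a spawn of attached audio on the tile.
; A timeline then drives the tile's relative height, rotation and scale each update.
; When the timeline finishes, the actor is destroyed.
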(event EventBeginPlay
  (bind _sm_tile (Variables|Default|GetSM_Tile))
  (Variables|Default|SetInitialPosition (Variables|/Script/Engine.SceneComponent|GetRelativeLocation _sm_tile))
  (Variables|Default|SetInitialRotation (Variables|/Script/Engine.SceneComponent|GetRelativeRotation _sm_tile))
  (Audio|SpawnSoundAttached 0 (Variables|Default|GetSM_Tile))
  (bind (_direction _height _rotation _scale) (Timeline|Timeline PlayFromStart)
    (:Update
      (Transformation|SetRelativeTransform (Variables|Default|GetSM_Tile) (Math|Transform|MakeTransform (Math|Vector|MakeVector 0.0 0.0 _height) (Math|Rotator|MakeRotator (.x _rotation) (.y _rotation) (.z _rotation)) _scale) false false))
    (:Finished
      (Actor|DestroyActor))))

(fn ConstructionScript)
